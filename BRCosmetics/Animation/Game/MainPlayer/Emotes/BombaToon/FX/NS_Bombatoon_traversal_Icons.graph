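# Niagara effect for the BombaToon traversal emote that emits cartoon sprite icons, including a banana, from the character's skeletal mesh.
# Looping bursts send icons along curve-driven gravity paths, with owner speed adjusting spawn rate, drag and world-to-local blending.
# Icon emitters write to a system death array, and death-driven emitters spawn ink splashes and pops at the death position.
# Layer colors follow time of day, and front-end modifiers adjust banana behavior.
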NiagaraSystem /BRCosmetics/Animation/Game/MainPlayer/Emotes/BombaToon/FX/NS_Bombatoon_traversal_Icons
dependencies {
 usedDataInterfaces [
  &NiagaraDataInterfaceArrayColor
  &NiagaraDataInterfaceCurve
  &NiagaraDataInterfaceVector2DCurve
  &NiagaraDataInterfaceSpriteRendererInfo
  &NiagaraDataInterfaceArrayFloat
  &NiagaraDataInterfaceCamera
  &NiagaraDataInterfaceSkeletalMesh
  &NiagaraDataInterfaceArrayPosition
 ]
 usedDynamicInputs []
 usedModules [
  &self:SystemSpawnScript.SystemScriptSource.SystemScriptGraph.NiagaraNodeAssignment_0.SetVariables_#1
  &self:SystemSpawnScript.SystemScriptSource.SystemScriptGraph.NiagaraNodeAssignment_2.SetVariables_#2
  &SystemState
  &self:ReadDeath
  &BlendSpaceHelper
  &self:SystemSpawnScript.SystemScriptSource.SystemScriptGraph.NiagaraNodeAssignment_1.SetVariables_#3
  &self:SystemSpawnScript.SystemScriptSource.SystemScriptGraph.NiagaraNodeAssignment_3.SetVariables_#4
  &self:SystemSpawnScript.SystemScriptSource.SystemScriptGraph.NiagaraNodeAssignment_4.SetVariables_#5
 ]
}
emitters {
 Banana {
  emitterSpawnScript [
   /SetVariables(
    Emitter.Layer 1 color:LinearColor=&FNTimeOfDayColor(Night Color=<rgba 0 0.0009 0.0156 1>, Morning Color=<rgba 0 0 0 1>, Day Color=<rgba 0 0 0 1>, Evening Color=<rgba 0 0 0 1>)
    Emitter.Layer 2 Color:LinearColor=&FNTimeOfDayColor(
     Night Color=<rgba 0.186 0.293 0.4 1>
     Morning Color=<rgba 0.7454 0.6867 0.6939 1>
     Day Color=<rgba 1 1 1 1>
     Evening Color=<rgba 0.349 0.3009 0.2977 1>
    )
    Emitter.Layer 3 Color:LinearColor=&FNTimeOfDayColor(Night Color=<rgba 0 0.0009 0.0156 1>, Morning Color=<rgba 0 0 0 1>, Day Color=<rgba 0 0 0 1>, Evening Color=<rgba 0 0 0 1>)
    Emitter.Layer 4 Color:LinearColor=&FNTimeOfDayColor(
     Night Color=<rgba 0.209 0.2945 0.38 1>
     Morning Color=<rgba 0.7454 0.6867 0.6939 1>
     Day Color=<rgba 1 1 1 1>
     Evening Color=<rgba 0.349 0.3009 0.2977 1>
    )
   ) {isSetParameters=true}
   /SetVariables(Emitter.Idle SpawnRate=1, Emitter.Traversal SpawnRate=1) {isSetParameters=true}
   /SetVariables(Emitter.Max Traversal speed=153, Emitter.Rail / Car Speed=500) {isSetParameters=true}
  ]
  emitterUpdateScript [
   &EmitterState(Life Cycle Mode=::NewEnumerator1, Loop Behavior=::NewEnumerator1, Loop Delay=2, Loop Duration=2.9, UseLoopDelay=true)
   &SpawnRate(
    SpawnRate=&FN_FrontEndFloatModifier(
     Gameplay=&MakeCustomFloatFromBool(
      Bool=&SetBoolByFloatComparison(A=&VectorLength(Vector=$User.OwnerLocalVelocity), B=139, Comparison Type=::NewEnumerator6)
      True Float=$Emitter.Idle SpawnRate
      False Float=&MakeCustomFloatFromBool(
       Bool=&SetBoolByFloatComparison(A=&VectorLength(Vector=$User.OwnerLocalVelocity), B=500, Comparison Type=::NewEnumerator6)
       True Float=$Emitter.Traversal SpawnRate
       False Float=Linked() {linkedVariable {name=Emitter.Car/rail SpawnRate, type=NiagaraFloat}}
      )
     )
     Front End=$Emitter.Traversal SpawnRate
    )
    Use Spawn Probability=true
   )
   /SetVariables(Emitter.Position2:NiagaraPosition=$Engine.Emitter.SimulationPosition) {isSetParameters=true}
  ]
  particleSpawnScript [
   /SetVariables(Particles.VisibilityTag=&SpawnGroup()) {isSetParameters=true}
   &InitializeParticle(
    Lifetime=1.5
    Mesh Scale Min=<2 2 2>
    Mesh Scale Mode=::NewEnumerator2
    Position Offset=<0 200 3>
    Sprite Size Mode=::NewEnumerator1
    Uniform Sprite Size Max=65
    Uniform Sprite Size Min=40
    UsePositionOffset=false
    Write Material Random=true
   )
   &SkeletalMeshLocation(
    Bone / Socket Index=2
    Bone Sampling Mode=::NewEnumerator4
    MaskSpawnGroup=false
    Sampled Orientation Coordinate Space=::Local
    Sampled Position Offset=&FN_FrontEndVector3Modifier(
     Gameplay=&Lerp_Vector(A=<0 35 25>, B=<0 100 25>, Alpha=$System.InterpolateOverTime.InterpolateOverTime_Float.Moving Average)
     Front End=<0 50 30>
    )
    Sampled Position Offset Coordinate Space:ENiagaraSkelMeshTransforms=::NewEnumerator2
    Skeletal Mesh=$System.Skeletal Mesh Reader
   )
   &CameraOffset(Camera Offset Amount=2)
   &AddVelocity(Velocity=&FN_FrontEndVector3Modifier(Gameplay=<0 0 0>, Front End=<0 -100 0>))
   &ShapeLocation(
    Disc Coverage=1
    MaskSpawnGroup=false
    Override Seed=false
    Ring / Disc Mode:ENiagaraRingDiscMode=::NewEnumerator1
    Shape Primitive=::NewEnumerator4
    UseOwnerScale=true
   )
   /SetVariables() {isSetParameters=true}
  ]
  particleUpdateScript [
   &ParticleState()
   &ScaleSpriteSize(Uniform Curve Sprite Scale=DataInterface() {propertyValues {Curve {keys [
        <at 0 0 RCIM_Cubic>
        {arriveTangent=-0.0027, interpMode=RCIM_Cubic, leaveTangent=-0.0027, tangentMode=RCTM_User, time=0.0604, value=1.3}
        {arriveTangent=-0.3324, interpMode=RCIM_Cubic, leaveTangent=-0.3324, time=0.1204, value=1}
        {arriveTangent=-0.2399, interpMode=RCIM_Cubic, leaveTangent=-0.2399, tangentMode=RCTM_User, time=0.963, value=1}
        <at 1 0 RCIM_Cubic>
       ]}}})
   &InheritSourceMovement(
    Applied Position Scale=&MakeCustomVectorFromBool(
     Bool=&SetBoolByFloatComparison(A=&VectorLength(Vector=$Engine.Owner.Velocity), B=1000)
     True Vector=<0.5 0.5 0.5>
     False Vector=<0 0 0>
    )
    ApplyForce=false
    ApplyPosition=true
    ApplyPositionFalloff=false
   )
   &SolveForcesAndVelocity(Clamp Velocity=false, Limit Acceleration=false)
   &SubUVAnimation(
    Output Alternate Renderer Binding=true
    "Sprite Renderer "=DataInterface() {propertyValues {SpriteRenderer=&self:Banana_0.NiagaraSpriteRendererProperties_7}}
    SubUV Lookup Index Scale=2
    UseEndFrame=false
    UseStartFrame=false
   )
   /SetVariables(Particles.SpriteAlignment=<0 0 1>) {isSetParameters=true}
   /SetVariables(
    Particles.layer 1:Vector4f=&MakeVector4(X=1, Y=1, Z=0, W=0)
    Particles.layer 2:Vector4f=&MakeVector4(X=0, Y=0, Z=1, W=1)
   ) {isSetParameters=true}
   &self:WriteDeath_04(Position Array:NiagaraDataInterfaceArrayPosition=$System.DeathArray) {module=WriteDeath_04}
   &BlendSpace("Position Blend (World <-> Local)":NiagaraFloat=&FN_FrontEndFloatModifier(
     Gameplay=0
     Front End=1
     Translation Contribution:NiagaraFloat=&FN_FrontEndFloatModifier(Gameplay=0, Front End=1)
     Rotation Contribution:NiagaraFloat=&FN_FrontEndFloatModifier(Gameplay=0, Front End=1)
    ))
  ]
  rendererClasses [&NiagaraSpriteRendererProperties]
  renderers [
   {rendererClass=&NiagaraSpriteRendererProperties}
   {rendererClass=&NiagaraSpriteRendererProperties, rendererIndex=1}
  ]
 }
 Icons_mid_body_parent006 {
  emitterSpawnScript [
   /SetVariables(Emitter.Idle Drag=1, Emitter.Rail drag=1, Emitter.Traversal Drag=3) {isSetParameters=true}
   /SetVariables(Emitter.Max Traversal speed=153.5, Emitter.Rail / Car Speed=500, Emitter.Traversal Particle Spawn Rate=1) {isSetParameters=true}
   /SetVariables(
    Emitter.Layer 1 color:LinearColor=&FNTimeOfDayColor(Night Color=<rgba 0 0.0003 0.0015 1>, Morning Color=<rgba 0 0 0 1>, Day Color=<rgba 0 0 0 1>, Evening Color=<rgba 0 0 0 1>)
    Emitter.Layer 2 Color:LinearColor=&FNTimeOfDayColor(
     Night Color=<rgba 0.1505 0.2311 0.35 1>
     Morning Color=<rgba 0.5104 0.4402 0.4354 1>
     Day Color=<rgba 1 1 1 1>
     Evening Color=<rgba 0.349 0.3009 0.2977 1>
    )
    Emitter.Layer 3 Color:LinearColor=&FNTimeOfDayColor(Night Color=<rgba 0 0.0003 0.0015 1>, Morning Color=<rgba 0 0 0 1>, Day Color=<rgba 0 0 0 1>, Evening Color=<rgba 0 0 0 1>)
    Emitter.Layer 4 Color:LinearColor=&FNTimeOfDayColor(
     Night Color=<rgba 0.1837 0.2669 0.35 1>
     Morning Color=<rgba 0.5104 0.4402 0.4354 1>
     Day Color=<rgba 1 1 1 1>
     Evening Color=<rgba 0.349 0.3009 0.2977 1>
    )
   ) {isSetParameters=true}
  ]
  emitterUpdateScript [
   &EmitterState(Life Cycle Mode=::NewEnumerator1, Loop Duration=5.87, Recalculate Duration Each Loop=true, UseLoopDelay=false)
   &SpawnBurst_Instantaneous(Spawn Group=3, Spawn Time=3.2, Use Spawn Probability=false, UseLoopCountLimit=false) {module=SpawnBurst_Instantaneous004}
   &SpawnBurst_Instantaneous(Spawn Group=2, Spawn Time=1.2, Use Spawn Probability=false, UseLoopCountLimit=false) {module=SpawnBurst_Instantaneous003}
   &SpawnBurst_Instantaneous(Spawn Group=1, Spawn Time=0.85, Use Spawn Probability=false, UseLoopCountLimit=false)
   &SpawnBurst_Instantaneous(Spawn Time=0.05, Use Spawn Probability=false, UseLoopCountLimit=false) {module=SpawnBurst_Instantaneous002}
  ]
  particleSpawnScript [
   /SetVariables(Particles.VisibilityTag=&SpawnGroup()) {isSetParameters=true, module=SetVariables_#6}
   &InitializeParticle(
    Lifetime=&MergeHelper_FloatSwitch(0=2, 1=2.2, 2=3, 3=2.5, 4=2.5, 5=2.5, 6=1.5, Index=&SpawnGroup(), Num=::NewEnumerator5)
    Mesh Scale Mode=::NewEnumerator3
    Mesh Uniform Scale=2.6199
    Sprite Size Mode=::NewEnumerator3
    Uniform Sprite Size=&MergeHelper_FloatSwitch(
     0=&RandomRangeFloat(Minimum=35, Maximum=50)
     1=&RandomRangeFloat(Minimum=35, Maximum=50)
     2=&RandomRangeFloat(Minimum=35, Maximum=50)
     3=&RandomRangeFloat(Minimum=35, Maximum=50)
     4=50
     5=40
     6=50
     Index=&SpawnGroup()
     Num=::NewEnumerator5
    )
    UsePositionOffset=false
    Write Material Random=true
   )
   &ApplyInitialForces()
   &SkeletalMeshLocation(
    Bone / Socket Index=2
    Bone Sampling Mode=::NewEnumerator4
    MaskSpawnGroup=false
    Sampled Position Offset=&MergeHelper_VectorSwitch(
     0=<0 -50 20>
     1=<0 -50 20>
     2=<0 -50 20>
     3=<0 -50 20>
     4=<0 -75 30>
     5=<0 -75 35>
     6=<0 -35 35>
     Index=&SpawnGroup()
     Num=::NewEnumerator5
    )
    Sampled Position Offset Coordinate Space:ENiagaraSkelMeshTransforms=::NewEnumerator2
    Skeletal Mesh=$System.Skeletal Mesh
   )
   &AddVelocity(Velocity=&MergeHelper_VectorSwitch(0=<0 0 0>, 1=<0 0 0>, 2=<0 0 0>, 3=<0 0 0>, 4=<0 0 0>, Index=&SpawnGroup(), Num=::NewEnumerator3)) {enabled=false}
   /SetVariables(Particles.ID par:NiagaraFloat=&MakeFloatFromInt(Int=&ReturnParticleIDAttributes())) {isSetParameters=true}
   /SetVariables(Particles.fake sprite NormalizedAge=0.5) {isSetParameters=true, module=SetVariables_#7}
   /SetVariables() {isSetParameters=true}
   &CameraOffset(Camera Offset Amount=5)
  ]
  particleUpdateScript [
   &ParticleState()
   &self:WriteDeath_01(Position Array:NiagaraDataInterfaceArrayPosition=$System.DeathArray) {module=WriteDeath_01}
   &ScaleSpriteSize(Uniform Curve Sprite Scale=DataInterface() {propertyValues {Curve {keys [
        {tangentMode=RCTM_User, time=0, value=0}
        {tangentMode=RCTM_User, time=0.0442, value=1}
        {interpMode=RCIM_Cubic, tangentMode=RCTM_Break, time=0.8913, value=1}
        {interpMode=RCIM_Cubic, tangentMode=RCTM_Break, time=0.9465, value=1.5}
       ]}}})
   &GravityForce(
    Coordinate Space=::Local
    Gravity=&MergeHelper_VectorSwitch(
     0=&MakeVector(
      X=&FloatFromCurve(
       FloatCurve=DataInterface() {propertyValues {Curve {keys [
           {interpMode=RCIM_Cubic, tangentMode=RCTM_Break, time=0, value=1}
           {interpMode=RCIM_Cubic, tangentMode=RCTM_Break, time=0.4986, value=-1}
           {interpMode=RCIM_Cubic, tangentMode=RCTM_Break, time=1, value=1}
          ]}}}
       Scale Curve=250
      )
      Y=-250
      Z=&FloatFromCurve(FloatCurve=DataInterface() {propertyValues {Curve {keys [<at 0 0>,<at 0.2 950>,<at 0.4976 -450>,<at 0.75 650>,<at 1 -250>]}}})
     )
     1=&MakeVector(
      X=&FloatFromCurve(
       FloatCurve=DataInterface() {propertyValues {Curve {keys [
           {interpMode=RCIM_Cubic, tangentMode=RCTM_Break, time=0, value=-1}
           {interpMode=RCIM_Cubic, tangentMode=RCTM_Break, time=0.4986, value=1}
           {interpMode=RCIM_Cubic, tangentMode=RCTM_Break, time=1, value=-1}
          ]}}}
       Scale Curve=350
      )
      Y=&FloatFromCurve(FloatCurve=DataInterface() {propertyValues {Curve {keys [<at 0 -350>,<at 0.5 -190>]}}})
      Z=&FloatFromCurve(FloatCurve=DataInterface() {propertyValues {Curve {keys [<at 0 0>,<at 0.2 950>,<at 0.4976 -450>,<at 0.75 650>,<at 1 -250>]}}})
     )
     2=&MakeVector(
      X=&FloatFromCurve(
       FloatCurve=DataInterface() {propertyValues {Curve {keys [
           {interpMode=RCIM_Cubic, tangentMode=RCTM_Break, time=0, value=1}
           {interpMode=RCIM_Cubic, tangentMode=RCTM_Break, time=0.4986, value=-1}
           {interpMode=RCIM_Cubic, tangentMode=RCTM_Break, time=1, value=1}
          ]}}}
       Scale Curve=350
      )
      Y=-250
      Z=&FloatFromCurve(FloatCurve=DataInterface() {propertyValues {Curve {keys [<at 0 0>,<at 0.2915 833.5294>,<at 0.4976 -450>,<at 0.75 650>,<at 1 -250>]}}})
     )
     3=&MakeVector(
      X=&FloatFromCurve(
       FloatCurve=DataInterface() {propertyValues {Curve {keys [
           {interpMode=RCIM_Cubic, tangentMode=RCTM_Break, time=0, value=-1}
           {interpMode=RCIM_Cubic, tangentMode=RCTM_Break, time=0.4986, value=1}
           {interpMode=RCIM_Cubic, tangentMode=RCTM_Break, time=1, value=-1}
          ]}}}
       Scale Curve=350
      )
      Y=-190
      Z=&FloatFromCurve(
       FloatCurve=DataInterface() {propertyValues {Curve {keys [<at 0 0>,<at 0.2 850>,<at 0.4976 -450>,<at 0.75 650>,<at 1 -250>]}}}
       Scale Curve=0.5
      )
     )
     4=<0 0 0>
     5=<0 0 0>
     6=<0 0 0>
     Index=&SpawnGroup()
     Num=::NewEnumerator5
    )
   )
   &Drag(
    Drag=&Multiply_Float(
     A=&MergeHelper_FloatSwitch(
      0=3.5
      1=3.5
      2=&FloatFromCurve(FloatCurve=DataInterface() {propertyValues {Curve {keys [<at 0 0 RCIM_Constant>,<at 0.1054 0.9294 RCIM_Constant>]}}}, Scale Curve=5.5)
      3=4
      4=3
      5=3
      6=3
      Index=&SpawnGroup()
      Num=::NewEnumerator5
     )
     B=&MakeCustomFloatFromBool(
      Bool=&SetBoolByFloatComparison(A=&VectorLength(Vector=$User.OwnerLocalVelocity), B=139, Comparison Type=::NewEnumerator6)
      True Float=$Emitter.Idle Drag
      False Float=&MakeCustomFloatFromBool(
       Bool=&SetBoolByFloatComparison(A=&VectorLength(Vector=$User.OwnerLocalVelocity), B=500, Comparison Type=::NewEnumerator6)
       True Float=$Emitter.Traversal Drag
       False Float=$Emitter.Rail drag
      )
     )
    )
    Use Linear Drag=true
    Use Rotational Drag=true
   )
   &Drag(Drag=3, Use Linear Drag=true, Use Rotational Drag=true) {enabled=false, module=Drag001}
   &SolveForcesAndVelocity(Clamp Velocity=false, Limit Acceleration=false)
   /SetVariables(
    Particles.Burg Position:NiagaraPosition=&TransformVector(
     Vector=&AddVectorToPosition(Position=$Particles.SkeletalMeshLocation.SampledPosition, Vector=<0 -5 8.8567>)
     Source Space=::Simulation
     Destination Space=::Local
     Transform Vector as Position=true
    )
    Particles.Burg SpriteSize:Vector2f=&Vector2DFromFloat(Value=40.1217)
    Particles.Layer 1 DynParam 0:Vector4f=&MakeVector4(X=1, Y=1, Z=0, W=0)
    Particles.Layer 1 DynParam 1:Vector4f=&MakeVector4(X=1, Y=1, Z=0, W=0)
   ) {isSetParameters=true, module=SetVariables_#8}
   /SetVariables(
    Particles. DynParam 0:Vector4f=&MakeVector4(X=0, Y=0, Z=1, W=1)
    Particles. DynParam 1:Vector4f=&MakeVector4(X=0, Y=0, Z=1, W=1)
    Particles.Ink Drip 2 Position:NiagaraPosition=&TransformVector(
     Vector=&AddVectorToPosition(Position=$Particles.SkeletalMeshLocation.SampledPosition, Vector=<0 -5 8.7201>)
     Source Space=::Simulation
     Destination Space=::Local
     Transform Vector as Position=true
    )
    Particles.Ink Drip 2 SpriteSize:Vector2f=&Vector2DFromFloat(Value=35)
   ) {isSetParameters=true, module=SetVariables_#9}
   &SubUVAnimation(
    End Frame Range Override=11
    Output Alternate Renderer Binding=true
    "Sprite Renderer "=DataInterface() {propertyValues {SpriteRenderer=&self:Icons_mid_body_parent006_0.NiagaraSpriteRendererProperties_0}}
    SubUV Lookup Index:NiagaraFloat=&FloatFromCurve(FloatCurve=DataInterface() {propertyValues {Curve {defaultValue=0, keys [<at 0 0>,<at 0.25 1>,<at 0.5 0>,<at 0.75 1>,<at 1 0>]}}})
    UseEndFrame=true
    UseStartFrame=true
   ) {module=SubUVAnimation002}
   &SubUVAnimation(
    End Frame Range Override=10
    Output Alternate Renderer Binding=true
    "Sprite Renderer "=DataInterface() {propertyValues {SpriteRenderer=&self:Icons_mid_body_parent006_0.NiagaraSpriteRendererProperties_5}}
    SubUV Lookup Index:NiagaraFloat=&FloatFromCurve(FloatCurve=DataInterface() {propertyValues {Curve {defaultValue=1, keys [<at 0 0>,<at 0.5 1>,<at 1 0>]}}})
    SubUV Lookup Index Scale=2
    UseEndFrame=true
    UseStartFrame=true
   )
   &SubUVAnimation(
    End Frame Range Override=9
    Output Alternate Renderer Binding=true
    "Sprite Renderer "=DataInterface() {propertyValues {SpriteRenderer=&self:Icons_mid_body_parent006_0.NiagaraSpriteRendererProperties_4}}
    SubUV Lookup Index:NiagaraFloat=&FloatFromCurve(FloatCurve=DataInterface() {propertyValues {Curve {defaultValue=1, keys [<at 0 0>,<at 0.5 1>,<at 1 0>]}}})
    UseEndFrame=true
    UseStartFrame=false
   ) {module=SubUVAnimation001}
   &SubUVAnimation(
    End Frame Range Override=10
    Output Alternate Renderer Binding=true
    "Sprite Renderer "=DataInterface() {propertyValues {SpriteRenderer=&self:Icons_mid_body_parent006_0.NiagaraSpriteRendererProperties_1}}
    SubUV Lookup Index:NiagaraFloat=&FloatFromCurve(FloatCurve=DataInterface() {propertyValues {Curve {defaultValue=0, keys [<at 0 0>,<at 0.5 1>,<at 1 0>]}}})
    SubUV Lookup Index Scale=1.2
    UseEndFrame=true
    UseStartFrame=true
   ) {module=SubUVAnimation004}
   &SubUVAnimation(
    End Frame Range Override=16
    Output Alternate Renderer Binding=true
    "Sprite Renderer "=DataInterface() {propertyValues {SpriteRenderer=&self:Icons_mid_body_parent006_0.NiagaraSpriteRendererProperties_7}}
    SubUV Lookup Index Scale=1.3
    UseEndFrame=true
    UseStartFrame=true
   ) {module=SubUVAnimation005}
   /SetVariables(Particles.Heart SpriteRotation:NiagaraFloat=&Multiply_Float(
     A=&Lerp_Float(A=-20, B=20, Alpha=$Particles.NormalizedAge)
     B=&SelectFloatFromArray(
      Array Sampling Mode=::NewEnumerator0
      Direct Array Index=?
      Float Selection Array=DataInterface() {propertyValues {FloatData [-1,1]}}
      Random Seed=0
     )
    )) {isSetParameters=true, module=SetVariables_#10}
   /SetVariables(Particles.fake SpriteSize:Vector2f=<2 2>) {isSetParameters=true, module=SetVariables_#11}
   /SetVariables(Particles.fake sprite new age:NiagaraFloat=&Subtract_Float(A=$Particles.fake sprite NormalizedAge, B=0)) {isSetParameters=true, module=SetVariables_#12}
   /SetVariables(Particles.fade out flower:LinearColor=&MakeLinearColorFromVectorAndFloat("Vector (RGB)"=<1 1 1>, "Float (Alpha)"=1)) {isSetParameters=true}
   &ScaleColor(
    Scale Alpha=/Niagara/DynamicInputs/Camera/CameraFade(
     Camera Query:NiagaraDataInterfaceCamera=DataInterface()
     Fade Offset:NiagaraFloat=120
     Fade Distance:NiagaraFloat=200
     Use Owner Position:NiagaraBool=true
    )
    ScaleA=true
    ScaleRGB=true
   )
   &SubUVAnimation(
    End Frame Range Override=16
    Output Alternate Renderer Binding=true
    "Sprite Renderer "=DataInterface() {propertyValues {SpriteRenderer=&self:Icons_mid_body_parent006_0.NiagaraSpriteRendererProperties_2}}
    SubUV Lookup Index Scale=1.3
    UseEndFrame=true
    UseStartFrame=true
   ) {module=SubUVAnimation003}
   /SetVariables(Particles.Banana SpriteAlignment:Vector3f=<0 0 1>) {isSetParameters=true}
   &BlendSpace("Position Blend (World <-> Local)":NiagaraFloat=&MakeCustomFloatFromBool(
     Bool=&SetBoolByFloatComparison(
      A=&VectorLength(Vector=$User.OwnerLocalVelocity)
      B=$Emitter.Max Traversal speed
      Comparison Type=::NewEnumerator6
     )
     True Float=&FloatFromCurve(FloatCurve=DataInterface() {propertyValues {Curve {
         defaultValue=0
         keys [
          {interpMode=RCIM_Cubic, tangentMode=RCTM_User, time=0, value=1}
          {arriveTangent=-7.4988, interpMode=RCIM_Cubic, leaveTangent=-7.4988, tangentMode=RCTM_User, time=0.5, value=0}
         ]
        }}})
     False Float=&MakeCustomFloatFromBool(
      Bool=true
      True Float=&FloatFromCurve(FloatCurve=DataInterface() {propertyValues {Curve {defaultValue=0, keys [{interpMode=RCIM_Cubic, tangentMode=RCTM_User, time=0, value=1}]}}})
      False Float=&FloatFromCurve(FloatCurve=DataInterface() {propertyValues {Curve {defaultValue=0, keys [{interpMode=RCIM_Cubic, tangentMode=RCTM_User, time=0, value=1}]}}})
     )
     Translation Contribution:NiagaraFloat=1
     Rotation Contribution:NiagaraFloat=1
    ))
  ]
  rendererClasses [&NiagaraSpriteRendererProperties]
  renderers [
   {rendererClass=&NiagaraSpriteRendererProperties}
   {rendererClass=&NiagaraSpriteRendererProperties, rendererIndex=1}
   {rendererClass=&NiagaraSpriteRendererProperties, rendererIndex=2}
   {rendererClass=&NiagaraSpriteRendererProperties, rendererIndex=3}
  ]
 }
 SplashElements_feet {
  emitterSpawnScript [
   /SetVariables(Emitter.highlight=DataInterface() {propertyValues {GpuSyncMode=SyncCpuToGpu, MaxElements=0, ColorData [<rgba 1 1 1 1>]}}) {isSetParameters=true}
   /SetVariables(
    Emitter.Layer 1 color:LinearColor=&FNTimeOfDayColor(Night Color=<rgba 0 0.0009 0.0156 1>, Morning Color=<rgba 0 0 0 1>, Day Color=<rgba 0 0 0 1>, Evening Color=<rgba 0 0 0 1>)
    Emitter.Layer 2 Color:LinearColor=&FNTimeOfDayColor(Night Color=<rgba 0.7581 0.8368 1 1>, Morning Color=<rgba 1 1 1 1>, Day Color=<rgba 1 1 1 1>, Evening Color=<rgba 1 1 1 1>)
    Emitter.Layer 3 Color:LinearColor=&FNTimeOfDayColor(Night Color=<rgba 0 0.0009 0.0156 1>, Morning Color=<rgba 0 0 0 1>, Day Color=<rgba 0 0 0 1>, Evening Color=<rgba 0 0 0 1>)
    Emitter.Layer 4 Color:LinearColor=&FNTimeOfDayColor(Night Color=<rgba 0.7581 0.8368 1 1>, Morning Color=<rgba 1 1 1 1>, Day Color=<rgba 1 1 1 1>, Evening Color=<rgba 1 1 1 1>)
   ) {isSetParameters=true}
   /SetVariables() {isSetParameters=true}
  ]
  emitterUpdateScript [
   &EmitterState(Inactive Response=::NewEnumerator1, Life Cycle Mode=::NewEnumerator1, Loop Delay=0.5, Loop Duration=5.87, UseLoopDelay=false)
   &SpawnPerFrame(Spawn:NiagaraBool=$System.DeathSpawn, Spawn Group=2, Use Spawn Probability=false) {module=SpawnPerFrame002}
   &SpawnPerFrame(Spawn:NiagaraBool=$System.DeathSpawn, Spawn Group=1, Use Spawn Probability=false) {module=SpawnPerFrame001}
   &SpawnPerFrame(Spawn:NiagaraBool=$System.DeathSpawn, Spawn Count=10, Use Spawn Probability=false)
  ]
  particleSpawnScript [
   /SetVariables(Particles.VisibilityTag=&SpawnGroup()) {isSetParameters=true}
   &InitializeParticle(
    Color=<rgba 0 0 0 1>
    Lifetime=&MergeHelper_FloatSwitch(
     0=&RandomRangeFloat(Minimum=0.125, Maximum=0.55)
     1=&RandomRangeFloat(Minimum=0.125, Maximum=0.55)
     2=&RandomRangeFloat(Minimum=0.125, Maximum=0.4)
     3=?
     Index=&SpawnGroup()
     Num=::NewEnumerator1
    )
    Mass Mode=::NewEnumerator3
    Mesh Scale=&Multiply_VectorByFloat(Vector=<2 1 1>, Float=&RandomRangeFloat(Minimum=0.3, Maximum=2))
    Mesh Scale Mode=::NewEnumerator4
    Position Offset=<0 -35 0>
    Sprite Size=<2 2>
    Sprite Size Mode=::NewEnumerator4
    UsePositionOffset=false
    Write Material Random=true
   )
   &SubUVAnimation(End Frame=3, SubUV Animation Mode:ENiagara_SubUVLookupMode=::NewEnumerator2)
   &AddVelocity(
    Constrain To Radius=0.5952
    Module.Use Falloff Distance=false
    Override Seed=false
    Velocity Mode=::NewEnumerator1
    Velocity Speed=&RandomRangeFloat(Minimum=300, Maximum=350)
   )
   &ShapeLocation(MaskSpawnGroup=false, Override Seed=false, Shape Origin=$System.DeathPosition, Sphere Radius=5, UseOwnerScale=true)
   &AddVelocityInCone(
    Cone Angle=71.9164
    Cone Axis=<0 0 1>
    Module.Use Velocity Falloff On Cone Axis=false
    Velocity Strength=&RandomRangeFloat(Minimum=250, Maximum=500)
   ) {enabled=false}
  ]
  particleUpdateScript [
   &ParticleState()
   &GravityForce(
    Coordinate Space=::Local
    Gravity=&MergeHelper_VectorSwitch(
     0=&MakeVector(X=0, Y=0, Z=&RandomRangeFloat(Minimum=-350, Maximum=-980))
     1=&MakeVector(X=0, Y=0, Z=&RandomRangeFloat(Minimum=-150, Maximum=-350))
     2=<0 0 0>
     3=<0 0 0>
     4=<0 0 0>
     Index=&SpawnGroup()
     Num=::NewEnumerator3
    )
   )
   &ScaleColor(
    Scale Alpha=&FloatFromCurve(FloatCurve=DataInterface() {propertyValues {Curve {keys [<at -0.0046 1.0113 RCIM_Cubic>,<at 1.0011 0.002 RCIM_Cubic>]}}})
    ScaleA=true
    ScaleRGB=true
   )
   &InheritSourceMovement(
    Applied Force Scale=&VectorFromFloat(Value=&FloatFromCurve(FloatCurve=DataInterface() {propertyValues {Curve {keys [{arriveTangent=0.4365, interpMode=RCIM_Cubic, leaveTangent=0.4365, tangentMode=RCTM_User, time=0, value=1}, <at 1 0 RCIM_Cubic>]}}}))
    Applied Position Scale=&VectorFromFloat(Value=&FloatFromCurve(FloatCurve=DataInterface() {propertyValues {Curve {keys [<at 0 1 RCIM_Cubic>,<at 0.9 0 RCIM_Cubic>]}}}))
    ApplyForce=true
    ApplyForceFalloff=false
    ApplyPosition=true
    ApplyPositionFalloff=false
    Module.LimitForceSpeed=false
   ) {enabled=false}
   &Drag(
    Drag=&RandomRangeFloat(Minimum=4, Maximum=8, Evaluation Type=::NewEnumerator1)
    Use Linear Drag=true
    Use Rotational Drag=false
   )
   &SolveForcesAndVelocity(Clamp Velocity=false, Limit Acceleration=false)
   &ScaleMeshSizeBySpeed(
    Sample Scale Factor By Curve=true
    Scale Factor Curve=DataInterface() {propertyValues {Curve {defaultValue=0, keys [{tangentMode=RCTM_User, time=0, value=1}, {tangentMode=RCTM_User, time=1, value=0}]}}}
   )
   &ScaleSpriteSizeBySpeed(
    Max Scale Factor=&MergeHelper_Vector2Switch(0=<2 2.3>, 1=<1 1>, 2=<1 1>, 3=<0 0>, 4=<0 0>, Index=&SpawnGroup(), Num=::NewEnumerator3)
    Sample Scale Factor By Curve=true
    Scale Factor Curve=DataInterface() {propertyValues {Curve {defaultValue=0, keys [{tangentMode=RCTM_User, time=0, value=1}, {tangentMode=RCTM_User, time=1, value=0}]}}}
    Velocity Threshold=500
   )
   /SetVariables(
    Particles.Heart SpriteSize:Vector2f=&Multiply_Vector2D_ByFloat(
     Float=&RandomRangeFloat(Minimum=0.5, Maximum=4.5)
     Vector2D=&Multiply_Vector2D(
      A=$Particles.SpriteSize
      B=&Vector2DFromCurve(
       Vector2Curve=DataInterface() {propertyValues {
         XCurve {defaultValue=1, keys [{tangentMode=RCTM_User, time=0, value=0}, {tangentMode=RCTM_User, time=0.3, value=1}]}
         YCurve {defaultValue=1, keys [{tangentMode=RCTM_User, time=0, value=0}, {tangentMode=RCTM_User, time=0.3, value=1}]}
        }}
       ScaleCurve=<1 1>
      )
     )
    )
    Particles.INK SpriteSize:Vector2f=&Multiply_Vector2D_ByFloat(Float=1.2662, Vector2D=&RandomRangeVector2D(Minimum=<1.9792 5>, Maximum=<3 5>))
    Particles.Star SpriteSize:Vector2f=&Multiply_Vector2D_ByFloat(Float=&RandomRangeFloat(Minimum=1, Maximum=3.5), Vector2D=$Particles.SpriteSize)
   ) {isSetParameters=true}
   &ScaleSpriteSize(
    Uniform Curve Scale=&RandomRangeFloat(Minimum=1, Maximum=3)
    Uniform Curve Sprite Scale=DataInterface() {propertyValues {Curve {keys [
        <at 0 0 RCIM_Cubic>
        {arriveTangent=1, interpMode=RCIM_Cubic, leaveTangent=1, time=0.1, value=1}
        {arriveTangent=-1.1111, interpMode=RCIM_Cubic, leaveTangent=-1.1111, time=0.7, value=0.7}
        <at 1 0 RCIM_Cubic>
       ]}}}
   )
   &ScaleMeshSize(Scale Factor=&VectorFromFloat(Value=&FloatFromCurve(
      FloatCurve=DataInterface() {propertyValues {Curve {keys [<at 0 1 RCIM_Cubic>, {arriveTangent=-1, interpMode=RCIM_Cubic, leaveTangent=-1, time=0.3, value=1.125}, <at 1 0 RCIM_Cubic>]}}}
      Scale Curve=0.6
     )))
   &OffsetPosition(Position Offset=&MakeVector(
     X=0
     Y=&FloatFromCurve(
      FloatCurve=DataInterface() {propertyValues {Curve {defaultValue=1, keys [{tangentMode=RCTM_User, time=0, value=0}, {tangentMode=RCTM_User, time=1, value=1}]}}}
      Scale Curve=-35
     )
     Z=0
    )) {enabled=false}
   /SetVariables(
    Particles.Toon Liquid up SpriteSize:Vector2f=&Multiply_Vector2D_ByFloat(
     Float=1
     Vector2D=&Vector2DFromCurve(
      Vector2Curve=DataInterface() {propertyValues {
        XCurve {keys [
          {arriveTangent=-2.9854, interpMode=RCIM_Cubic, time=0, value=1}
          {arriveTangent=0.475, interpMode=RCIM_Cubic, leaveTangent=0.475, time=0.2, value=0.1886}
          <at 1 1.475 RCIM_Cubic>
         ]}
        YCurve {keys [
          <at 0 1 RCIM_Cubic>
          {arriveTangent=71.1029, interpMode=RCIM_Cubic, leaveTangent=71.1029, tangentMode=RCTM_User, time=0.2, value=6}
          {arriveTangent=-16.0372, interpMode=RCIM_Cubic, leaveTangent=-16.0372, tangentMode=RCTM_User, time=0.5, value=5.5}
          <at 1 0 RCIM_Cubic>
         ]}
       }}
      ScaleCurve=<10 6>
     )
    )
    Particles.Toon liquid up Color:LinearColor=<rgba 0 0 0 1>
    Particles.Toon liquid up DynParam:Vector4f=&MakeVector4(
     X=0
     Y=&FloatFromCurve(FloatCurve=DataInterface() {propertyValues {Curve {keys [<at 0 0.675>,<at 1.1 0.5>]}}})
     Z=0
     W=0
    )
   ) {isSetParameters=true}
   /SetVariables(Particles.Color=&MakeLinearColorFromVectorAndFloat("Vector (RGB)"=&MakeVectorFromLinearColorRGB(Color=$Emitter.Layer 1 color), "Float (Alpha)"=1)) {isSetParameters=true}
   &BlendSpace("Position Blend (World <-> Local)":NiagaraFloat=&MakeCustomFloatFromBool(
     Bool=&SetBoolByFloatComparison(A=&VectorLength(Vector=$Engine.Owner.Velocity), B=400)
     True Float=1
     False Float=&FloatFromCurve(FloatCurve=DataInterface() {propertyValues {Curve {keys [<at 0 1 RCIM_Cubic>, {arriveTangent=-12.1775, interpMode=RCIM_Cubic, leaveTangent=-12.1775, tangentMode=RCTM_User, time=0.25, value=0}]}}})
     Translation Contribution:NiagaraFloat=1
     Rotation Contribution:NiagaraFloat=0
    )) {module=BlendSpace001}
   &CameraOffset()
   /SetVariables(Particles.HEART SpriteAlignment:Vector3f=&Multiply_Vector(A=&RandomRangeVector(Minimum=<0 0 1>, Maximum=<1 1 1>), B=$Particles.Velocity)) {isSetParameters=true}
  ]
  rendererClasses [&NiagaraMeshRendererProperties, &NiagaraSpriteRendererProperties]
  renderers [
   {rendererClass=&NiagaraMeshRendererProperties, rendererIndex=0}
   {rendererClass=&NiagaraSpriteRendererProperties, rendererIndex=1}
   {rendererClass=&NiagaraSpriteRendererProperties, rendererIndex=2}
   {rendererClass=&NiagaraSpriteRendererProperties, rendererIndex=3}
   {rendererClass=&NiagaraSpriteRendererProperties, rendererIndex=4}
  ]
 }
 death001 {
  emitterSpawnScript [/SetVariables(Emitter.Color Array=DataInterface() {propertyValues {GpuSyncMode=SyncCpuToGpu, MaxElements=0, ColorData [<rgba 1 1 1 1>, <rgba 0 0 0 0>, <rgba 0 0 0 0>, <rgba 0 0 0 0>]}}) {isSetParameters=true}]
  emitterUpdateScript [
   &EmitterState(Life Cycle Mode=::NewEnumerator1, Loop Duration=5.87, UseLoopDelay=false)
   &SpawnBurst_Instantaneous(Use Spawn Probability=false, UseLoopCountLimit=false) {enabled=false}
   &SpawnPerFrame(Spawn:NiagaraBool=$System.DeathSpawn, Use Spawn Probability=false) {module=SpawnPerFrame001}
  ]
  particleSpawnScript [
   /SetVariables(Particles.VisibilityTag=&SpawnGroup()) {isSetParameters=true}
   /SetVariables(Particles.InitSpriteRotRate=&UniformRangedFloat(Minimum=-1, Maximum=1)) {isSetParameters=true}
   &InitializeParticle(
    Color Mode=::NewEnumerator0
    Lifetime=0.2
    Mesh Scale=&Multiply_VectorByFloat(Vector=<2 1 1>, Float=&RandomRangeFloat(Minimum=1, Maximum=2))
    Mesh Scale Mode=::NewEnumerator4
    Position=$System.DeathPosition
    Position Mode=::NewEnumerator1
    Sprite Rotation Mode=::NewEnumerator2
    Sprite Size Mode=::NewEnumerator3
    Uniform Sprite Size=&RandomRangeFloat(Minimum=40, Maximum=55)
    UsePositionOffset=false
    Write Material Random=true
   )
   &ShapeLocation(
    MaskSpawnGroup=false
    Non Uniform Scale=<0.1 0.1 0.1>
    Shape Origin=$System.DeathPosition
    Sphere Distribution=::NewEnumerator1
    Sphere Radius=0
    UseOwnerScale=true
   )
   &AddVelocity(Velocity=&MergeHelper_VectorSwitch(
     0=<0 0 0>
     1=&RandomRangeVector(Minimum=<-401.3117 -150 -150>, Maximum=<150 150 150>)
     2=<0 0 0>
     3=<0 0 0>
     4=<0 0 0>
     Index=0
     Num=::NewEnumerator3
    ))
   &ShapeLocation(MaskSpawnGroup=false, Override Seed=false, Shape Origin=$System.DeathPosition, Sphere Radius=5, UseOwnerScale=true) {enabled=false, module=ShapeLocation001}
  ]
  particleUpdateScript [
   &ParticleState()
   &ScaleSpriteSize(Uniform Curve Sprite Scale=DataInterface() {propertyValues {Curve {keys [<at -0.0006 0>, {tangentMode=RCTM_User, time=0.5, value=0.9176}]}}})
   &SpriteRotationRate(Rotation Rate=&MergeHelper_FloatSwitch(
     0=&Multiply_Float(
      A=$Particles.InitSpriteRotRate
      B=&FloatFromCurve(
       FloatCurve=DataInterface() {propertyValues {Curve {keys [
           {arriveTangent=-1.9214, interpMode=RCIM_Cubic, leaveTangent=-1.9214, tangentMode=RCTM_User, time=-0.0046, value=1.0113}
           {arriveTangent=-0.2496, interpMode=RCIM_Cubic, leaveTangent=-0.2496, tangentMode=RCTM_User, time=0.9896, value=0.16}
          ]}}}
       CurveIndex=$Emitter.Age
      )
     )
     1=0
     2=5
     3=?
     Index=&SpawnGroup()
     Num=::NewEnumerator1
    ))
   &Drag(
    Drag=&MergeHelper_FloatSwitch(
     0=&RandomRangeFloat(Minimum=0, Maximum=0)
     1=&RandomRangeFloat(Minimum=4, Maximum=8, Evaluation Type=::NewEnumerator1)
     2=&RandomRangeFloat(Minimum=1, Maximum=8)
     3=&RandomRangeFloat(Minimum=0, Maximum=1)
     Index=&SpawnGroup()
     Num=::NewEnumerator0
    )
    Use Linear Drag=true
    Use Rotational Drag=true
   )
   &GravityForce(Gravity=&MergeHelper_VectorSwitch(
     0=<0 0 0>
     1=&MakeVector(X=0, Y=0, Z=&RandomRangeFloat(Minimum=-350, Maximum=-980))
     2=&RandomRangeVector(Minimum=<0 0 50>, Maximum=<0 0 150>)
     3=&RandomRangeVector(Minimum=<0 0 -50>, Maximum=<1 1 -650>)
     Index=&SpawnGroup()
     Num=::NewEnumerator0
    ))
   &SolveForcesAndVelocity(Clamp Velocity=false, Limit Acceleration=false)
   /SetVariables(Particles.Ink Droplets Color:LinearColor=&MakeLinearColorFromVectorAndFloat(
     "Vector (RGB)"=&MakeVectorFromLinearColorRGB(Color=&FNTimeOfDayColor(Night Color=<rgba 0 0 0 1>, Morning Color=<rgba 0 0 0 1>, Day Color=<rgba 0 0 0 1>, Evening Color=<rgba 0 0 0 1>))
     "Float (Alpha)"=1
    )) {isSetParameters=true}
   &SubUVAnimation(
    Define SubUV Setup Manually:ENiagara_AutomaticManual=::NewEnumerator1
    Direct Frame Index:NiagaraFloat=&SelectFloatFromArray(
     Array Sampling Mode=::NewEnumerator2
     Direct Array Index=?
     Interpolation=$Particles.NormalizedAge
     Float Selection Array=DataInterface() {propertyValues {FloatData [4,5,6]}}
    )
    End Frame Range Override=6
    Output Alternate Renderer Binding=true
    Start Frame Range Override=4
    SubUV Animation Mode:ENiagara_SubUVLookupModeV2=::NewEnumerator5
    UseEndFrame=true
    UseStartFrame=true
   )
   &Color(Color=&MakeLinearColorFromVectorAndFloat(
     "Vector (RGB)"=<1 1 1>
     "Float (Alpha)"=&FloatFromCurve(FloatCurve=DataInterface() {propertyValues {Curve {keys [<at 0.0031 0>,<at 0.0162 1.0051>,<at 0.7047 0.9881>,<at 0.9621 -0.0118>]}}})
    ))
   /SetVariables(
    Particles. DynParam 0:Vector4f=&MakeVector4(X=0, Y=0, Z=1, W=1)
    Particles. DynParam 1:Vector4f=&MakeVector4(X=0, Y=0, Z=1, W=1)
    Particles.Ink Drip 2 SpriteSize:Vector2f=&Vector2DFromFloat(Value=35)
   ) {isSetParameters=true}
   &BlendSpace("Position Blend (World <-> Local)":NiagaraFloat=&MakeCustomFloatFromBool(
     Bool=&SetBoolByFloatComparison(A=&VectorLength(Vector=$Engine.Owner.Velocity), B=400)
     True Float=1
     False Float=&FloatFromCurve(FloatCurve=DataInterface() {propertyValues {Curve {keys [<at 0 1 RCIM_Cubic>, {arriveTangent=-12.1775, interpMode=RCIM_Cubic, leaveTangent=-12.1775, tangentMode=RCTM_User, time=0.25, value=0}]}}})
     Translation Contribution:NiagaraFloat=1
     Rotation Contribution:NiagaraFloat=0
    )) {module=BlendSpace001}
   &CameraOffset(Camera Offset Amount=7)
  ]
  rendererClasses [&NiagaraSpriteRendererProperties, &NiagaraMeshRendererProperties]
  renderers [
   {rendererClass=&NiagaraSpriteRendererProperties}
   {rendererClass=&NiagaraMeshRendererProperties, rendererIndex=1}
  ]
 }
 fake_spawn_in {
  emitterSpawnScript [
   /SetVariables(Emitter.Idle Drag=1, Emitter.Rail drag=3, Emitter.Traversal Drag=3) {isSetParameters=true}
   /SetVariables(Emitter.Max Traversal speed=153.5, Emitter.Rail / Car Speed=500, Emitter.Traversal Particle Spawn Rate=1) {isSetParameters=true}
   /SetVariables(
    Emitter.Layer 1 color:LinearColor=&FNTimeOfDayColor(Night Color=<rgba 0 0.0003 0.0015 1>, Morning Color=<rgba 0 0 0 1>, Day Color=<rgba 0 0 0 1>, Evening Color=<rgba 0 0 0 1>)
    Emitter.Layer 2 Color:LinearColor=&FNTimeOfDayColor(Night Color=<rgba 0.7581 0.8368 1 1>, Morning Color=<rgba 1 1 1 1>, Day Color=<rgba 1 1 1 1>, Evening Color=<rgba 1 1 1 1>)
    Emitter.Layer 3 Color:LinearColor=&FNTimeOfDayColor(Night Color=<rgba 0 0.0003 0.0015 1>, Morning Color=<rgba 0 0 0 1>, Day Color=<rgba 0 0 0 1>, Evening Color=<rgba 0 0 0 1>)
    Emitter.Layer 4 Color:LinearColor=&FNTimeOfDayColor(Night Color=<rgba 0.7581 0.8368 1 1>, Morning Color=<rgba 1 1 1 1>, Day Color=<rgba 1 1 1 1>, Evening Color=<rgba 1 1 1 1>)
   ) {isSetParameters=true}
  ]
  emitterUpdateScript [
   &EmitterState(Life Cycle Mode=::NewEnumerator1, Loop Duration=5.87, Recalculate Duration Each Loop=true, UseLoopDelay=false)
   &SpawnBurst_Instantaneous(Spawn Group=3, Spawn Time=3.1, Use Spawn Probability=false, UseLoopCountLimit=false) {module=SpawnBurst_Instantaneous004}
   &SpawnBurst_Instantaneous(Spawn Group=2, Spawn Time=1.3, Use Spawn Probability=false, UseLoopCountLimit=false) {module=SpawnBurst_Instantaneous003}
   &SpawnBurst_Instantaneous(Spawn Group=1, Spawn Time=0.8, Use Spawn Probability=false, UseLoopCountLimit=false)
   &SpawnBurst_Instantaneous(Spawn Time=0.1, Use Spawn Probability=false, UseLoopCountLimit=false) {module=SpawnBurst_Instantaneous002}
  ]
  particleSpawnScript [
   /SetVariables(Particles.VisibilityTag=&SpawnGroup()) {isSetParameters=true, module=SetVariables_#6}
   &InitializeParticle(
    Lifetime=&MergeHelper_FloatSwitch(0=0.1, 1=0.1, 2=0.1, 3=0.1, 4=0.05, Index=&SpawnGroup(), Num=::NewEnumerator3)
    Mesh Scale Mode=::NewEnumerator3
    Mesh Uniform Scale=2.6199
    Sprite Size Mode=::NewEnumerator3
    Uniform Sprite Size=0.1
    UsePositionOffset=false
    Write Material Random=true
   )
   &ApplyInitialForces()
   &SkeletalMeshLocation(
    Bone / Socket Index=2
    Bone Sampling Mode=::NewEnumerator4
    MaskSpawnGroup=false
    Sampled Position Offset=&MergeHelper_VectorSwitch(
     0=<0 -50 20>
     1=<0 -50 20>
     2=<0 -50 20>
     3=<0 -50 20>
     4=<0 -75 30>
     5=<0 -75 35>
     6=<0 -35 35>
     Index=&SpawnGroup()
     Num=::NewEnumerator5
    )
    Sampled Position Offset Coordinate Space:ENiagaraSkelMeshTransforms=::NewEnumerator2
    Skeletal Mesh=$System.Skeletal Mesh
   )
   &AddVelocity(Velocity=&MergeHelper_VectorSwitch(0=<0 0 0>, 1=<0 0 0>, 2=<0 0 0>, 3=<0 0 0>, 4=<0 0 0>, Index=&SpawnGroup(), Num=::NewEnumerator3)) {enabled=false}
   /SetVariables() {isSetParameters=true}
   /SetVariables(Particles.fake sprite NormalizedAge=0.5) {isSetParameters=true, module=SetVariables_#7}
  ]
  particleUpdateScript [
   &ParticleState()
   &Drag(Use Linear Drag=true, Use Rotational Drag=true)
   &GravityForce(
    Coordinate Space=::Local
    Gravity=&MergeHelper_VectorSwitch(
     0=&MakeVector(
      X=&FloatFromCurve(
       FloatCurve=DataInterface() {propertyValues {Curve {keys [
           {interpMode=RCIM_Cubic, tangentMode=RCTM_Break, time=0, value=1}
           {interpMode=RCIM_Cubic, tangentMode=RCTM_Break, time=0.4986, value=-1}
           {interpMode=RCIM_Cubic, tangentMode=RCTM_Break, time=1, value=1}
          ]}}}
       Scale Curve=250
      )
      Y=-250
      Z=&FloatFromCurve(FloatCurve=DataInterface() {propertyValues {Curve {keys [<at 0 0>,<at 0.2 950>,<at 0.4976 -450>,<at 0.75 650>,<at 1 -250>]}}})
     )
     1=&MakeVector(
      X=&FloatFromCurve(
       FloatCurve=DataInterface() {propertyValues {Curve {keys [
           {interpMode=RCIM_Cubic, tangentMode=RCTM_Break, time=0, value=-1}
           {interpMode=RCIM_Cubic, tangentMode=RCTM_Break, time=0.4986, value=1}
           {interpMode=RCIM_Cubic, tangentMode=RCTM_Break, time=1, value=-1}
          ]}}}
       Scale Curve=350
      )
      Y=&FloatFromCurve(FloatCurve=DataInterface() {propertyValues {Curve {keys [<at 0 -350>,<at 0.5 -190>]}}})
      Z=&FloatFromCurve(FloatCurve=DataInterface() {propertyValues {Curve {keys [<at 0 0>,<at 0.2 950>,<at 0.4976 -450>,<at 0.75 650>,<at 1 -250>]}}})
     )
     2=&MakeVector(
      X=&FloatFromCurve(
       FloatCurve=DataInterface() {propertyValues {Curve {keys [
           {interpMode=RCIM_Cubic, tangentMode=RCTM_Break, time=0, value=1}
           {interpMode=RCIM_Cubic, tangentMode=RCTM_Break, time=0.4986, value=-1}
           {interpMode=RCIM_Cubic, tangentMode=RCTM_Break, time=1, value=1}
          ]}}}
       Scale Curve=350
      )
      Y=-250
      Z=&FloatFromCurve(FloatCurve=DataInterface() {propertyValues {Curve {keys [<at 0 0>,<at 0.2915 833.5294>,<at 0.4976 -450>,<at 0.75 650>,<at 1 -250>]}}})
     )
     3=&MakeVector(
      X=&FloatFromCurve(
       FloatCurve=DataInterface() {propertyValues {Curve {keys [
           {interpMode=RCIM_Cubic, tangentMode=RCTM_Break, time=0, value=-1}
           {interpMode=RCIM_Cubic, tangentMode=RCTM_Break, time=0.4986, value=1}
           {interpMode=RCIM_Cubic, tangentMode=RCTM_Break, time=1, value=-1}
          ]}}}
       Scale Curve=350
      )
      Y=-190
      Z=&FloatFromCurve(
       FloatCurve=DataInterface() {propertyValues {Curve {keys [<at 0 0>,<at 0.2 850>,<at 0.4976 -450>,<at 0.75 650>,<at 1 -250>]}}}
       Scale Curve=0.5
      )
     )
     4=<0 0 0>
     5=<0 0 0>
     6=<0 0 0>
     Index=&SpawnGroup()
     Num=::NewEnumerator5
    )
   )
   &SolveForcesAndVelocity(Clamp Velocity=false, Limit Acceleration=false)
   &SubUVAnimation(
    End Frame Range Override=10
    Output Alternate Renderer Binding=true
    "Sprite Renderer "=DataInterface() {propertyValues {SpriteRenderer=&self:fake_spawn_in_0.NiagaraSpriteRendererProperties_0}}
    SubUV Lookup Index:NiagaraFloat=&FloatFromCurve(FloatCurve=DataInterface() {propertyValues {Curve {defaultValue=1, keys [<at 0 0>,<at 0.5 1>,<at 1 0>]}}})
    UseEndFrame=true
    UseStartFrame=true
   ) {module=SubUVAnimation002}
   &SubUVAnimation(
    Direct Frame Index:NiagaraFloat=&FloatFromCurve(FloatCurve=DataInterface() {propertyValues {Curve {keys [<at 0 1 RCIM_Constant>,<at 0.1 2 RCIM_Constant>,<at 0.2 3 RCIM_Constant>,<at 0.3 5 RCIM_Constant>,<at 0.4 6 RCIM_Constant>,<at 0.5 7 RCIM_Constant>,<at 0.6 4 RCIM_Constant>]}}})
    Output Alternate Renderer Binding=true
    "Sprite Renderer "=DataInterface() {propertyValues {SpriteRenderer=&self:fake_spawn_in_0.NiagaraSpriteRendererProperties_2}}
    SubUV Animation Mode:ENiagara_SubUVLookupModeV2=::NewEnumerator5
   ) {module=SubUVAnimation003}
   &OffsetPosition(Position Offset=&MakeVector(
     X=0
     Y=&FloatFromCurve(
      FloatCurve=DataInterface() {propertyValues {Curve {
         defaultValue=1
         keys [{interpMode=RCIM_Cubic, tangentMode=RCTM_User, time=0, value=0}, {interpMode=RCIM_Cubic, tangentMode=RCTM_User, time=1, value=1}]
        }}}
      Scale Curve=-150
     )
     Z=0
    )) {enabled=false}
   /SetVariables(
    Particles.Burg Position:NiagaraPosition=&TransformVector(
     Vector=&AddVectorToPosition(Position=$Particles.SkeletalMeshLocation.SampledPosition, Vector=<0 -5 8.8567>)
     Source Space=::Simulation
     Destination Space=::Local
     Transform Vector as Position=true
    )
    Particles.Burg SpriteSize:Vector2f=&Vector2DFromFloat(Value=40.1217)
    Particles.Layer 1 DynParam 0:Vector4f=&MakeVector4(X=1, Y=1, Z=0, W=0)
    Particles.Layer 1 DynParam 1:Vector4f=&MakeVector4(X=1, Y=1, Z=0, W=0)
   ) {isSetParameters=true, module=SetVariables_#8}
   /SetVariables(
    Particles. DynParam 0:Vector4f=&MakeVector4(X=0, Y=0, Z=1, W=1)
    Particles. DynParam 1:Vector4f=&MakeVector4(X=0, Y=0, Z=1, W=1)
    Particles.Ink Drip 2 Position:NiagaraPosition=&TransformVector(
     Vector=&AddVectorToPosition(Position=$Particles.SkeletalMeshLocation.SampledPosition, Vector=<0 -5 8.7201>)
     Source Space=::Simulation
     Destination Space=::Local
     Transform Vector as Position=true
    )
    Particles.Ink Drip 2 SpriteSize:Vector2f=&Vector2DFromFloat(Value=35)
   ) {isSetParameters=true, module=SetVariables_#9}
   &SubUVAnimation(
    End Frame Range Override=10
    Output Alternate Renderer Binding=true
    "Sprite Renderer "=DataInterface() {propertyValues {SpriteRenderer=&self:fake_spawn_in_0.NiagaraSpriteRendererProperties_5}}
    SubUV Lookup Index:NiagaraFloat=&FloatFromCurve(FloatCurve=DataInterface() {propertyValues {Curve {defaultValue=1, keys [<at 0 0>,<at 0.5 1>,<at 1 0>]}}})
    SubUV Lookup Index Scale=0.8
    UseEndFrame=true
    UseStartFrame=false
   )
   &SubUVAnimation(
    End Frame Range Override=9
    Output Alternate Renderer Binding=true
    "Sprite Renderer "=DataInterface() {propertyValues {SpriteRenderer=&self:fake_spawn_in_0.NiagaraSpriteRendererProperties_4}}
    SubUV Lookup Index:NiagaraFloat=&FloatFromCurve(FloatCurve=DataInterface() {propertyValues {Curve {defaultValue=1, keys [<at 0 0>,<at 0.5 1>,<at 1 0>]}}})
    SubUV Lookup Index Scale=1.1
    UseEndFrame=true
    UseStartFrame=false
   ) {module=SubUVAnimation001}
   &SubUVAnimation(
    End Frame Range Override=10
    Output Alternate Renderer Binding=true
    "Sprite Renderer "=DataInterface() {propertyValues {SpriteRenderer=&self:fake_spawn_in_0.NiagaraSpriteRendererProperties_1}}
    SubUV Lookup Index:NiagaraFloat=&FloatFromCurve(FloatCurve=DataInterface() {propertyValues {Curve {defaultValue=1, keys [<at 0 0>,<at 0.5 1>,<at 1 0>]}}})
    UseEndFrame=true
    UseStartFrame=true
   ) {module=SubUVAnimation004}
   &SubUVAnimation(
    Output Alternate Renderer Binding=true
    "Sprite Renderer "=DataInterface() {propertyValues {SpriteRenderer=&self:fake_spawn_in_0.NiagaraSpriteRendererProperties_7}}
    UseEndFrame=false
    UseStartFrame=false
   ) {module=SubUVAnimation005}
   /SetVariables(Particles.Heart SpriteRotation:NiagaraFloat=&Multiply_Float(
     A=&Lerp_Float(A=-8, B=8, Alpha=$Particles.NormalizedAge)
     B=&SelectFloatFromArray(
      Array Sampling Mode=::NewEnumerator0
      Direct Array Index=?
      Float Selection Array=DataInterface() {propertyValues {FloatData [-1,1]}}
      Random Seed=0
     )
    )) {isSetParameters=true, module=SetVariables_#10}
   /SetVariables(Particles.fake SpriteSize:Vector2f=<2 2>) {isSetParameters=true, module=SetVariables_#11}
   /SetVariables(Particles.fake sprite new age:NiagaraFloat=&Subtract_Float(A=$Particles.fake sprite NormalizedAge, B=0)) {isSetParameters=true, module=SetVariables_#12}
   &SubUVAnimation(
    End Frame Range Override=10
    Output Alternate Renderer Binding=true
    SubUV Lookup Index:NiagaraFloat=&FloatFromCurve(FloatCurve=DataInterface() {propertyValues {Curve {defaultValue=1, keys [<at 0 0>,<at 0.5 1>,<at 1 0>]}}}, Scale Curve=0.5)
    UseEndFrame=true
    UseStartFrame=true
   ) {module=SubUVAnimation006}
   &SubUVAnimation(End Frame Range Override=14, Output Alternate Renderer Binding=true, UseEndFrame=true, UseStartFrame=true) {module=SubUVAnimation007}
   &self:WriteDeath_01(Position Array:NiagaraDataInterfaceArrayPosition=$System.DeathArray) {module=WriteDeath_01}
   &BlendSpace("Position Blend (World <-> Local)":NiagaraFloat=&MakeCustomFloatFromBool(
     Bool=&SetBoolByFloatComparison(
      A=&VectorLength(Vector=$User.OwnerLocalVelocity)
      B=$Emitter.Max Traversal speed
      Comparison Type=::NewEnumerator6
     )
     True Float=&FloatFromCurve(FloatCurve=DataInterface() {propertyValues {Curve {
         defaultValue=0
         keys [
          {interpMode=RCIM_Cubic, tangentMode=RCTM_User, time=0, value=1}
          {arriveTangent=-7.4988, interpMode=RCIM_Cubic, leaveTangent=-7.4988, tangentMode=RCTM_User, time=0.5, value=0}
         ]
        }}})
     False Float=&MakeCustomFloatFromBool(
      Bool=&SetBoolByFloatComparison(
       A=&VectorLength(Vector=$User.OwnerLocalVelocity)
       B=Linked() {linkedVariable {name=Emitter.Rail / Car Speed, type=NiagaraFloat}}
       Comparison Type=::NewEnumerator6
      )
      True Float=&FloatFromCurve(FloatCurve=DataInterface() {propertyValues {Curve {
          defaultValue=0
          keys [{interpMode=RCIM_Cubic, tangentMode=RCTM_User, time=0, value=1}, {interpMode=RCIM_Cubic, tangentMode=RCTM_User, time=0.15, value=0}]
         }}})
      False Float=&FloatFromCurve(FloatCurve=DataInterface() {propertyValues {Curve {
          defaultValue=0
          keys [{interpMode=RCIM_Cubic, tangentMode=RCTM_User, time=0, value=1}, {interpMode=RCIM_Cubic, tangentMode=RCTM_User, time=1, value=0.5}]
         }}})
     )
     Translation Contribution:NiagaraFloat=1
     Rotation Contribution:NiagaraFloat=1
    ))
  ]
  rendererClasses [&NiagaraSpriteRendererProperties]
  renderers [
   {rendererClass=&NiagaraSpriteRendererProperties}
   {rendererClass=&NiagaraSpriteRendererProperties, rendererIndex=1}
   {rendererClass=&NiagaraSpriteRendererProperties, rendererIndex=2}
   {rendererClass=&NiagaraSpriteRendererProperties, rendererIndex=3}
  ]
 }
}
summary {
 systemName=NS_Bombatoon_traversal_Icons
 emitters [
  {
   emitterName=SplashElements_feet
   rendererClasses [&NiagaraMeshRendererProperties, &NiagaraSpriteRendererProperties]
  }
  {emitterName=fake_spawn_in, rendererClasses [&NiagaraSpriteRendererProperties]}
  {emitterName=death001, rendererClasses [&NiagaraSpriteRendererProperties, &NiagaraMeshRendererProperties]}
  {emitterName=Banana, rendererClasses [&NiagaraSpriteRendererProperties]}
  {emitterName=Icons_mid_body_parent006, rendererClasses [&NiagaraSpriteRendererProperties]}
 ]
 userVariables [
  User.bIsFrontEnd {type=NiagaraBool}
  User.bIsFrontEndPreview {type=NiagaraBool}
  User.bisHeadHidden {type=NiagaraBool}
  User.FrontEnd {type=NiagaraFloat, defaultValue=1, description=""}
  User.OwnerLocalVelocity {type=Vector3f, description="", defaultValue {struct=Vector3f, value=<0 0 0>}}
 ]
}
