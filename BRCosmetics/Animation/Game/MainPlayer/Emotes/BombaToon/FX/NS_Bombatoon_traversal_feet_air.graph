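# Niagara effect for the BombaToon traversal emote that leaves dark ink-like splats and splashes around the character's feet, sampled from skeletal mesh bones or sockets.
# Decal spawns continuously, switching rate and ring radius by owner speed across idle, traversal and rail/car tiers.
# Decal002 bursts ground splats, using a slope trace outside front-end preview.
# SplashElements001 bursts mesh and sprite droplets that fall under gravity, inherit movement and blend local/world position by speed.
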NiagaraSystem /BRCosmetics/Animation/Game/MainPlayer/Emotes/BombaToon/FX/NS_Bombatoon_traversal_feet_air
dependencies {
 usedDataInterfaces [
  &NiagaraDataInterfaceCurve
  &NiagaraDataInterfaceCollisionQuery
  &NiagaraDataInterfaceDebugDraw
  &NiagaraDataInterfaceSkeletalMesh
 ]
 usedDynamicInputs []
 usedModules [
  &self:SystemSpawnScript.SystemScriptSource.SystemScriptGraph.NiagaraNodeAssignment_0.SetVariables_#1
  &SystemState
  &BlendSpaceHelper
 ]
}
emitters {
 Decal {
  emitterSpawnScript [
   /SetVariables(Emitter.Car/rail Spawn Rate=25, Emitter.Idle Spawn Rate=25, Emitter.Traversal Spawn Rate=25) {isSetParameters=true, module=SetVariables_#2}
   /SetVariables(Emitter.Max Traversal speed=153, Emitter.Rail / Car Speed=500) {isSetParameters=true, module=SetVariables_#3}
   /SetVariables(Emitter.Extra splatter Idle radius=35, Emitter.Idle Spawn Radius=100, Emitter.Traversal Spawn Radius=85, Emitter.rail/car Spawn Radius=50) {isSetParameters=true, module=SetVariables_#4}
  ]
  emitterUpdateScript [
   &EmitterState()
   &SpawnRate(
    SpawnGroup=1
    SpawnRate=&FloatFromCurve(
     FloatCurve=DataInterface() {propertyValues {Curve {keys [<at 0 0 RCIM_Constant>,<at 0.1 1 RCIM_Constant>,<at 1 1>]}}}
     CurveIndex=$Emitter.NormalizedLoopAge
     Scale Curve=&MakeCustomFloatFromBool(
      Bool=&SetBoolByFloatComparison(
       A=&VectorLength(Vector=$User.OwnerLocalVelocity)
       B=$Emitter.Max Traversal speed
       Comparison Type=::NewEnumerator6
      )
      True Float=$Emitter.Idle Spawn Rate
      False Float=&MakeCustomFloatFromBool(
       Bool=&SetBoolByFloatComparison(
        A=&VectorLength(Vector=$User.OwnerLocalVelocity)
        B=Linked() {linkedVariable {name=Emitter.Rail / Car Speed, type=NiagaraFloat}}
        Comparison Type=::NewEnumerator6
       )
       True Float=$Emitter.Traversal Particle Spawn Rate
       False Float=Linked() {linkedVariable {name=Emitter.Rail/ Car Spawn Rate, type=NiagaraFloat}}
      )
     )
    )
    Use Spawn Probability=false
   )
   &SpawnRate(
    SpawnGroup=1
    SpawnRate=&FloatFromCurve(
     FloatCurve=DataInterface() {propertyValues {Curve {keys [<at 0 0 RCIM_Constant>,<at 0.1 1 RCIM_Constant>,<at 1 1>]}}}
     CurveIndex=$Emitter.NormalizedLoopAge
     Scale Curve=&MakeCustomFloatFromBool(
      Bool=&SetBoolByFloatComparison(
       A=&VectorLength(Vector=$User.OwnerLocalVelocity)
       B=$Emitter.Max Traversal speed
       Comparison Type=::NewEnumerator6
      )
      True Float=$Emitter.Idle Spawn Rate
      False Float=&MakeCustomFloatFromBool(
       Bool=&SetBoolByFloatComparison(
        A=&VectorLength(Vector=$User.OwnerLocalVelocity)
        B=Linked() {linkedVariable {name=Emitter.Rail / Car Speed, type=NiagaraFloat}}
        Comparison Type=::NewEnumerator6
       )
       True Float=$Emitter.Traversal Particle Spawn Rate
       False Float=Linked() {linkedVariable {name=Emitter.Rail/ Car Spawn Rate, type=NiagaraFloat}}
      )
     )
    )
    Use Spawn Probability=false
   ) {module=SpawnRate001}
  ]
  particleSpawnScript [
   /SetVariables(Particles.VisibilityTag=&SpawnGroup()) {isSetParameters=true}
   &PartitionParticles()
   &InitializeParticle(
    Color=&FNTimeOfDayColor(Night Color=<rgba 0 0.0009 0.0156 1>, Morning Color=<rgba 0 0 0 1>, Day Color=<rgba 0 0 0 1>, Evening Color=<rgba 0 0 0 1>)
    Lifetime=&MergeHelper_FloatSwitch(
     0=&FN_FrontEndFloatModifier(Gameplay=&RandomRangeFloat(Minimum=0.55, Maximum=0.85), Front End=0)
     1=&FN_FrontEndFloatModifier(Gameplay=&RandomRangeFloat(Minimum=0.3, Maximum=0.7), Front End=0)
     2=?
     3=?
     Index=&SpawnGroup()
     Num=::NewEnumerator0
    )
    Mass Max=1.5
    Mass Mode=::NewEnumerator3
    Sprite Rotation Mode=::NewEnumerator2
    Sprite Size=<50 50>
    Sprite Size Mode=::NewEnumerator4
    UsePositionOffset=false
    Write Material Random=true
   )
   &SkeletalMeshLocation(
    Bone / Socket Index=$Particles.PartitionParticles.Partition
    Bone Sampling Mode=::NewEnumerator4
    MaskSpawnGroup=false
    Skeletal Mesh=$System.Skeletal Mesh
   )
   &ShapeLocation(
    Disc Coverage=0.85
    MaskSpawnGroup=false
    Offset Coordinate Space=::Local
    Offset Mode=::NewEnumerator0
    Override Seed=false
    Ring Radius=&MergeHelper_FloatSwitch(
     0=&MakeCustomFloatFromBool(
      Bool=&SetBoolByFloatComparison(
       A=&VectorLength(Vector=$User.OwnerLocalVelocity)
       B=$Emitter.Max Traversal speed
       Comparison Type=::NewEnumerator6
      )
      True Float=$Emitter.Idle Spawn Radius
      False Float=&MakeCustomFloatFromBool(
       Bool=&SetBoolByFloatComparison(
        A=&VectorLength(Vector=$User.OwnerLocalVelocity)
        B=Linked() {linkedVariable {name=Emitter.Rail / Car Speed, type=NiagaraFloat}}
        Comparison Type=::NewEnumerator6
       )
       True Float=$Emitter.Traversal Spawn Radius
       False Float=Linked() {linkedVariable {name=Emitter.rail/car Spawn Radius, type=NiagaraFloat}}
      )
     )
     1=&MakeCustomFloatFromBool(
      Bool=&SetBoolByFloatComparison(
       A=&VectorLength(Vector=$User.OwnerLocalVelocity)
       B=$Emitter.Max Traversal speed
       Comparison Type=::NewEnumerator6
      )
      True Float=$Emitter.Extra splatter Idle radius
      False Float=&MakeCustomFloatFromBool(
       Bool=&SetBoolByFloatComparison(
        A=&VectorLength(Vector=$User.OwnerLocalVelocity)
        B=Linked() {linkedVariable {name=Emitter.Rail / Car Speed, type=NiagaraFloat}}
        Comparison Type=::NewEnumerator6
       )
       True Float=$Emitter.Extra splatter traversal radius
       False Float=Linked() {linkedVariable {name=Emitter.Extra Splatter Rail/car radius, type=NiagaraFloat}}
      )
     )
     2=0
     3=0
     4=0
     Index=&SpawnGroup()
     Num=::NewEnumerator3
    )
    Shape Primitive=::NewEnumerator4
    UseOwnerScale=true
   )
   /SetVariables() {isSetParameters=true, module=SetVariables_#5}
  ]
  particleUpdateScript [
   &ParticleState()
   &ScaleColor(
    Scale Alpha=&FloatFromCurve(
     FloatCurve=DataInterface() {propertyValues {Curve {
        defaultValue=0
        keys [
         {interpMode=RCIM_Cubic, tangentMode=RCTM_User, time=0, value=0}
         {arriveTangent=0.2521, interpMode=RCIM_Cubic, leaveTangent=0.2521, tangentMode=RCTM_User, time=0.1147, value=1}
         {interpMode=RCIM_Cubic, tangentMode=RCTM_User, time=0.5, value=1}
         {arriveTangent=0.1836, interpMode=RCIM_Cubic, leaveTangent=0.1836, tangentMode=RCTM_User, time=0.7609, value=0.9806}
         {tangentMode=RCTM_User, time=1, value=0}
        ]
       }}}
     Scale Curve=3
    )
    Scale RGB=&MakeVectorFromLinearColorRGB(Color=<rgba 0 0 0 1>)
    ScaleA=true
    ScaleRGB=true
   )
   /SetVariables(Particles.DecalSize:Vector3f=&Multiply_Vector(
     A=&VectorFromFloat(Value=&RandomRangeFloat(Minimum=5, Maximum=27.5))
     B=&VectorFromFloat(Value=&FloatFromCurve(
       FloatCurve=DataInterface() {propertyValues {Curve {keys [
           {arriveTangent=4.1785, interpMode=RCIM_Cubic, leaveTangent=4.1785, tangentMode=RCTM_User, time=0, value=0.5}
           {arriveTangent=11.9214, interpMode=RCIM_Cubic, leaveTangent=11.9214, tangentMode=RCTM_User, time=0.0129, value=1.1398}
           {arriveTangent=0.0224, interpMode=RCIM_Cubic, leaveTangent=0.0224, tangentMode=RCTM_User, time=0.0291, value=1}
          ]}}}
       Scale Curve=&RandomRangeFloat(Minimum=2, Maximum=3)
      ))
    )) {isSetParameters=true, module=SetVariables_#6}
   /SetVariables(Particles.DecalFade:NiagaraFloat=&FloatFromCurve(FloatCurve=DataInterface() {propertyValues {Curve {keys [<at 0 0>,<at 0.0622 0.9378>,<at 1 0>]}}})) {isSetParameters=true, module=SetVariables_#7}
   /SetVariables(Particles.SpriteSize=&Multiply_Vector2D_ByFloat(
     Float=&FloatFromCurve(FloatCurve=DataInterface() {propertyValues {Curve {defaultValue=1, keys [<at 0.2181 1 RCIM_Constant>,<at 0.8164 1.5 RCIM_Constant>]}}})
     Vector2D=<20 20>
    )) {isSetParameters=true}
   /SetVariables(
    Particles. DynParam 0:Vector4f=&MakeVector4(X=0, Y=0, Z=1, W=1)
    Particles. DynParam 1:Vector4f=&MakeVector4(X=0, Y=0, Z=1, W=1)
    Particles.Ink Drip 2 Position:NiagaraPosition=&TransformVector(
     Vector=&AddVectorToPosition(Position=$Particles.SkeletalMeshLocation.SampledPosition, Vector=<0 -5 8.7201>)
     Source Space=::Simulation
     Destination Space=::Local
     Transform Vector as Position=true
    )
    Particles.Ink Drip 2 SpriteSize:Vector2f=&Vector2DFromFloat(Value=35)
   ) {isSetParameters=true}
   /SetVariables(
    Particles.Burg Position:NiagaraPosition=&TransformVector(
     Vector=&AddVectorToPosition(Position=$Particles.SkeletalMeshLocation.SampledPosition, Vector=<0 -5 8.8567>)
     Source Space=::Simulation
     Destination Space=::Local
     Transform Vector as Position=true
    )
    Particles.Burg SpriteSize:Vector2f=&Vector2DFromFloat(Value=40.1217)
    Particles.Layer 1 DynParam 0:Vector4f=&MakeVector4(X=1, Y=1, Z=0, W=0)
    Particles.Layer 1 DynParam 1:Vector4f=&MakeVector4(X=1, Y=1, Z=0, W=0)
   ) {isSetParameters=true}
   &Drag(
    Drag=&MergeHelper_FloatSwitch(0=0, 1=0, 2=3, 3=3, 4=0, Index=&SpawnGroup(), Num=::NewEnumerator3)
    Use Linear Drag=true
    Use Rotational Drag=true
   )
   &SolveForcesAndVelocity(Clamp Velocity=false, Limit Acceleration=false)
   /Niagara/Modules/Decal/Decal_Attributes(
    Color=1
    Decal Size:Vector3f=&MergeHelper_VectorSwitch(
     0=&Multiply_VectorByFloat(Vector=<5 5 5>, Float=15)
     1=&Multiply_VectorByFloat(Vector=<5 5 5>, Float=0)
     2=&Multiply_VectorByFloat(Vector=<5 5 5>, Float=&RandomRangeFloat(Minimum=0, Maximum=1))
     3=?
     Index=&SpawnGroup()
     Num=::NewEnumerator1
    )
    Fade=1
    Scale RGB:Vector3f=&Multiply_VectorByFloat(Vector=<1 1 1>, Float=1)
    ScaleA=true
    ScaleRGB=true
    Size=2
   )
  ]
  rendererClasses [&NiagaraDecalRendererProperties, &NiagaraSpriteRendererProperties]
  renderers [
   {rendererClass=&NiagaraDecalRendererProperties, rendererIndex=0}
   {rendererClass=&NiagaraDecalRendererProperties, rendererIndex=1}
   {rendererClass=&NiagaraDecalRendererProperties, rendererIndex=2}
   {rendererClass=&NiagaraSpriteRendererProperties, rendererIndex=3}
   {rendererClass=&NiagaraSpriteRendererProperties, rendererIndex=4}
  ]
 }
 Decal002 {
  emitterSpawnScript [
   /SetVariables(Emitter.Idle Spawn Rate=15) {isSetParameters=true, module=SetVariables_#2}
   /SetVariables(Emitter.Max Traversal speed=153, Emitter.Rail / Car Speed=500) {isSetParameters=true, module=SetVariables_#3}
   /SetVariables(
    Emitter.Extra splatter Idle radius=35
    Emitter.Idle Area spawn:Vector3f=&RandomRangeVector(Minimum=<100 150 15>, Maximum=<-150 -150 15>)
    Emitter.Idle Spawn Radius=85
    Emitter.Rail/ Car Area spawn:Vector3f=&RandomRangeVector(Minimum=<35 65 15>, Maximum=<-35 -65 15>)
    Emitter.Traversal Spawn Radius=85
    Emitter.Traversal area Spawn:Vector3f=<0 -65 15>
    Emitter.rail/car Spawn Radius=35
   ) {isSetParameters=true, module=SetVariables_#4}
  ]
  emitterUpdateScript [
   &EmitterState()
   &SpawnBurst_Instantaneous(
    Spawn Count=&MakeIntFromFloat(Float=&FN_FrontEndFloatModifier(Gameplay=50, Front End=15))
    Spawn Time=0.6
    Use Spawn Probability=false
    UseLoopCountLimit=false
   ) {module=SpawnBurst_Instantaneous001}
   &SpawnBurst_Instantaneous(
    Spawn Count=&MakeIntFromFloat(Float=&FN_FrontEndFloatModifier(Gameplay=35, Front End=5))
    Spawn Time=0.2
    Use Spawn Probability=false
    UseLoopCountLimit=false
   )
  ]
  particleSpawnScript [
   &PartitionParticles()
   &InitializeParticle(
    Color=&FNTimeOfDayColor(Night Color=<rgba 0 0.0009 0.0156 1>, Morning Color=<rgba 0 0 0 1>, Day Color=<rgba 0 0 0 1>, Evening Color=<rgba 0 0 0 1>)
    Lifetime=&FN_FrontEndFloatModifier(
     Gameplay=&RandomRangeFloat(Minimum=0.45, Maximum=0.7)
     Front End=&RandomRangeFloat(Minimum=0.45, Maximum=0.7)
    )
    Mass Max=1.5
    Mass Mode=::NewEnumerator3
    Sprite Rotation Mode=::NewEnumerator2
    Sprite Size=<1 1>
    Sprite Size Mode=::NewEnumerator4
    UsePositionOffset=false
    Write Material Random=true
   )
   &SkeletalMeshLocation(
    Bone / Socket Index=$Particles.PartitionParticles.Partition
    Bone Sampling Mode=::NewEnumerator4
    MaskSpawnGroup=false
    Skeletal Mesh=$System.Skeletal Mesh
   )
   &ShapeLocation(
    Disc Coverage=0.85
    MaskSpawnGroup=false
    Offset=&FN_FrontEndVector3Modifier(Gameplay=<0 0 0>, Front End=<-15 0 0>)
    Offset Coordinate Space=::Local
    Offset Mode=::NewEnumerator0
    Override Seed=false
    Ring Radius=&FN_FrontEndFloatModifier(
     Gameplay=&MakeCustomFloatFromBool(
      Bool=&SetBoolByFloatComparison(
       A=&VectorLength(Vector=$User.OwnerLocalVelocity)
       B=$Emitter.Max Traversal speed
       Comparison Type=::NewEnumerator6
      )
      True Float=$Emitter.Idle Spawn Radius
      False Float=&MakeCustomFloatFromBool(
       Bool=&SetBoolByFloatComparison(
        A=&VectorLength(Vector=$User.OwnerLocalVelocity)
        B=Linked() {linkedVariable {name=Emitter.Rail / Car Speed, type=NiagaraFloat}}
        Comparison Type=::NewEnumerator6
       )
       True Float=$Emitter.Traversal Spawn Radius
       False Float=Linked() {linkedVariable {name=Emitter.rail/car Spawn Radius, type=NiagaraFloat}}
      )
     )
     Front End=30
    )
    Shape Primitive=::NewEnumerator4
    UseOwnerScale=true
   )
   /SetVariables() {isSetParameters=true, module=SetVariables_#5}
   &CameraOffset(Camera Offset Amount=-5)
  ]
  particleUpdateScript [
   &ParticleState()
   &ScaleColor(
    Scale Alpha=&FloatFromCurve(
     FloatCurve=DataInterface() {propertyValues {Curve {
        defaultValue=0
        keys [
         {interpMode=RCIM_Cubic, tangentMode=RCTM_User, time=0, value=0}
         {arriveTangent=0.2521, interpMode=RCIM_Cubic, leaveTangent=0.2521, tangentMode=RCTM_User, time=0.1147, value=1}
         {interpMode=RCIM_Cubic, tangentMode=RCTM_User, time=0.5, value=1}
         {arriveTangent=0.1836, interpMode=RCIM_Cubic, leaveTangent=0.1836, tangentMode=RCTM_User, time=0.7609, value=0.9806}
         {tangentMode=RCTM_User, time=1, value=0}
        ]
       }}}
     Scale Curve=&FN_FrontEndFloatModifier(Gameplay=3, Front End=0)
    )
    Scale RGB=&MakeVectorFromLinearColorRGB(Color=<rgba 0 0 0 1>)
    ScaleA=true
    ScaleRGB=true
   )
   /SetVariables(Particles.DecalSize:Vector3f=&Multiply_Vector(
     A=&VectorFromFloat(Value=&RandomRangeFloat(Minimum=2, Maximum=15))
     B=&VectorFromFloat(Value=&FloatFromCurve(
       FloatCurve=DataInterface() {propertyValues {Curve {keys [
           {arriveTangent=4.1785, interpMode=RCIM_Cubic, leaveTangent=4.1785, tangentMode=RCTM_User, time=0, value=0.5}
           {arriveTangent=0.0224, interpMode=RCIM_Cubic, leaveTangent=0.0224, tangentMode=RCTM_User, time=0.05, value=1}
          ]}}}
       Scale Curve=&RandomRangeFloat(Minimum=1, Maximum=3)
      ))
    )) {isSetParameters=true, module=SetVariables_#6}
   /SetVariables(Particles.DecalFade:NiagaraFloat=&FloatFromCurve(FloatCurve=DataInterface() {propertyValues {Curve {keys [<at 0 1>,<at 1 0>]}}})) {isSetParameters=true, module=SetVariables_#7}
   &Decal_Attributes(
    Decal Facing Mode=1
    Decal Rotation Mode=1
    Decal Size:Vector3f=&Multiply_VectorByFloat(
     Vector=<5 5 5>
     Float=&FloatFromCurve(
      FloatCurve=DataInterface() {propertyValues {Curve {defaultValue=1, keys [{tangentMode=RCTM_User, time=0, value=0}, {tangentMode=RCTM_User, time=1, value=1}]}}}
      Scale Curve=15
     )
    )
    Fade=1
    Orientation=1
    Set Alignment=false
    Size=2
   )
   &self:Slope_Position_18(
    Apply Position:NiagaraBool=&InvertBool(Boolean:NiagaraBool=$User.bIsFrontEndPreview)
    Apply Rotation:NiagaraBool=&InvertBool(Boolean:NiagaraBool=$User.bIsFrontEndPreview)
    Fallback Position:NiagaraPosition=&AddVectorToPosition(Position=$Particles.Initial.Position, Vector=<1 1 1>)
    Height Adjustment/Offset=0.5
    Slope Position:NiagaraPosition=$Particles.SkeletalMeshLocation.SampledPosition
    Trace Length=120
    Trace Origin Height=85
   ) {module=Slope_Position_18}
   /SetVariables(
    Particles.Fake Decal SpriteSize:Vector2f=&Multiply_Vector2D_ByFloat(
     Float=&FloatFromCurve(
      FloatCurve=DataInterface() {propertyValues {Curve {
         defaultValue=1
         keys [
          {interpMode=RCIM_Cubic, tangentMode=RCTM_Break, time=0, value=0}
          {interpMode=RCIM_Cubic, tangentMode=RCTM_Break, time=0.1781, value=1.1}
          <at 0.1958 1 RCIM_Cubic>
         ]
        }}}
      Scale Curve=&RandomRangeFloat(Minimum=0.3, Maximum=0.8)
     )
     Vector2D=<50 50>
    )
    Particles.Fake decal SpriteRotation:NiagaraFloat=&RandomRangeFloat(Minimum=-1, Maximum=1)
    Particles.Fake decal color:LinearColor=&MakeLinearColorFromVectorAndFloat("Vector (RGB)"=<1 1 1>, "Float (Alpha)"=1)
   ) {isSetParameters=true}
  ]
  rendererClasses [
   &NiagaraDecalRendererProperties
   &NiagaraSpriteRendererProperties
   &NiagaraMeshRendererProperties
  ]
  renderers [
   {rendererClass=&NiagaraDecalRendererProperties, rendererIndex=0}
   {rendererClass=&NiagaraDecalRendererProperties, rendererIndex=1}
   {rendererClass=&NiagaraSpriteRendererProperties, rendererIndex=2}
   {rendererClass=&NiagaraSpriteRendererProperties, rendererIndex=3}
   {rendererClass=&NiagaraMeshRendererProperties, rendererIndex=4}
  ]
 }
 SplashElements001 {
  emitterUpdateScript [
   &EmitterState()
   &SpawnBurst_Instantaneous(Spawn Count=25, Spawn Group=1, Spawn Time=0.95, Use Spawn Probability=false, UseLoopCountLimit=false) {module=SpawnBurst_Instantaneous002}
   &SpawnBurst_Instantaneous(Spawn Count=15, Spawn Time=0.5, Use Spawn Probability=false, UseLoopCountLimit=false) {module=SpawnBurst_Instantaneous001}
   &SpawnBurst_Instantaneous(Spawn Count=15, Spawn Time=0.95, Use Spawn Probability=false, UseLoopCountLimit=false)
   &SpawnBurst_Instantaneous(Spawn Count=25, Spawn Group=1, Spawn Time=0.5, Use Spawn Probability=false, UseLoopCountLimit=false) {module=SpawnBurst_Instantaneous003}
  ]
  particleSpawnScript [
   /SetVariables(Particles.VisibilityTag=&SpawnGroup()) {isSetParameters=true}
   &PartitionParticles()
   &InitializeParticle(
    Color=&FNTimeOfDayColor(Night Color=<rgba 0 0.0009 0.0156 1>, Morning Color=<rgba 0 0 0 1>, Day Color=<rgba 0 0 0 1>, Evening Color=<rgba 0 0 0 1>)
    Lifetime=&MergeHelper_FloatSwitch(
     0=&RandomRangeFloat(Minimum=0.3, Maximum=0.65)
     1=&RandomRangeFloat(Minimum=0.4, Maximum=0.7)
     2=0
     3=0
     4=0
     Index=&SpawnGroup()
     Num=::NewEnumerator3
    )
    Mass Mode=::NewEnumerator3
    Mesh Scale=&Multiply_VectorByFloat(Vector=<2 1 1>, Float=&RandomRangeFloat(Minimum=1, Maximum=2))
    Mesh Scale Mode=::NewEnumerator4
    Sprite Size Max=<7 12>
    Sprite Size Min=<6 10>
    Sprite Size Mode=::NewEnumerator2
    UsePositionOffset=false
    Write Material Random=true
   )
   &AddVelocityInCone(
    Cone Angle=&FloatFromCurve(FloatCurve=DataInterface() {propertyValues {Curve {keys [<at 0 1>,<at 0.4546 0>]}}}, Scale Curve=150)
    Cone Axis=<0 0 0.5>
    Cone Axis Coordinate Space=::World
    Module.Use Velocity Falloff On Cone Axis=false
    Velocity Distribution Along Cone Axis=0.6
    Velocity Strength=&RandomRangeFloat(Minimum=50, Maximum=150)
   )
   &SkeletalMeshLocation(
    Bone / Socket Index=$Particles.PartitionParticles.Partition
    Bone Sampling Mode=::NewEnumerator4
    MaskSpawnGroup=false
    Orientation Sampling=::NewEnumerator1
    Sampled Orientation Coordinate Space=::Local
    Sampled Position Offset=&MergeHelper_VectorSwitch(0=<0 -7.5 -7>, 1=<3 -7.5 -7>, 2=?, 3=?, Index=&SpawnGroup(), Num=::NewEnumerator0)
    Sampled Position Offset Coordinate Space:ENiagaraSkelMeshTransforms=::NewEnumerator2
    Skeletal Mesh=$System.Skeletal Mesh
   )
   &SubUVAnimation(End Frame=3, SubUV Animation Mode:ENiagara_SubUVLookupMode=::NewEnumerator2)
   &ShapeLocation(MaskSpawnGroup=false, Override Seed=false, Sphere Radius=15, Sphere Surface Distribution=0.3762, UseOwnerScale=true)
  ]
  particleUpdateScript [
   &ParticleState()
   &GravityForce(
    Coordinate Space=::Local
    Gravity=&Multiply_VectorByFloat(
     Vector=&RandomRangeVector(Minimum=<0 0 -350>, Maximum=<0 0 -980>)
     Float=&FloatFromCurve(FloatCurve=DataInterface() {propertyValues {Curve {defaultValue=1, keys [{tangentMode=RCTM_User, time=0, value=0}, {tangentMode=RCTM_User, time=0.413, value=1}]}}})
    )
   )
   &Drag(Drag=0.1091, Use Linear Drag=true, Use Rotational Drag=false) {enabled=false}
   &ScaleColor(
    Scale Alpha=&FloatFromCurve(FloatCurve=DataInterface() {propertyValues {Curve {keys [<at -0.0046 1.0113 RCIM_Cubic>,<at 1.0011 0.002 RCIM_Cubic>]}}})
    ScaleA=true
    ScaleRGB=true
   )
   &InheritSourceMovement(
    Applied Force Scale=&VectorFromFloat(Value=&FloatFromCurve(FloatCurve=DataInterface() {propertyValues {Curve {keys [{arriveTangent=0.4365, interpMode=RCIM_Cubic, leaveTangent=0.4365, tangentMode=RCTM_User, time=0, value=1}, <at 1 0 RCIM_Cubic>]}}}))
    Applied Position Scale=&VectorFromFloat(Value=&FloatFromCurve(FloatCurve=DataInterface() {propertyValues {Curve {keys [<at 0 1 RCIM_Cubic>,<at 0.9 0 RCIM_Cubic>]}}}))
    ApplyForce=true
    ApplyForceFalloff=false
    ApplyPosition=true
    ApplyPositionFalloff=false
    Module.LimitForceSpeed=false
   )
   &GravityForce(Gravity=&MakeVector(
     X=0
     Y=0
     Z=&Multiply_Float(A=&RandomRangeFloat(Minimum=-500, Maximum=-980), B=0)
    )) {enabled=false, module=GravityForce001}
   &SolveForcesAndVelocity(Clamp Velocity=false, Limit Acceleration=false)
   &ScaleSpriteSize(Uniform Curve Sprite Scale=DataInterface() {propertyValues {Curve {keys [
        <at 0 0 RCIM_Cubic>
        {arriveTangent=1, interpMode=RCIM_Cubic, leaveTangent=1, time=0.1, value=1}
        {arriveTangent=-1.1111, interpMode=RCIM_Cubic, leaveTangent=-1.1111, time=0.7, value=0.7}
        <at 1 0 RCIM_Cubic>
       ]}}})
   &CameraOffset(Camera Offset Amount=&RandomRangeFloat(Minimum=0, Maximum=6))
   &ScaleMeshSize(Scale Factor=&VectorFromFloat(Value=&FloatFromCurve(FloatCurve=DataInterface() {propertyValues {Curve {keys [<at 0 1 RCIM_Cubic>, {arriveTangent=-1, interpMode=RCIM_Cubic, leaveTangent=-1, time=0.3, value=1.125}, <at 1 0 RCIM_Cubic>]}}})))
   &BlendSpace("Position Blend (World <-> Local)":NiagaraFloat=&MakeCustomFloatFromBool(
     Bool=&SetBoolByFloatComparison(A=&VectorLength(Vector=$Engine.Owner.Velocity), B=153.5)
     True Float=0
     False Float=&FloatFromCurve(FloatCurve=DataInterface() {propertyValues {Curve {keys [<at 0 1 RCIM_Cubic>, {arriveTangent=-12.1775, interpMode=RCIM_Cubic, leaveTangent=-12.1775, tangentMode=RCTM_User, time=0.25, value=0}]}}})
     Translation Contribution:NiagaraFloat=1
     Rotation Contribution:NiagaraFloat=0
    ))
  ]
  rendererClasses [&NiagaraMeshRendererProperties, &NiagaraSpriteRendererProperties]
  renderers [
   {rendererClass=&NiagaraMeshRendererProperties, rendererIndex=0}
   {rendererClass=&NiagaraSpriteRendererProperties, rendererIndex=1}
  ]
 }
}
summary {
 systemName=NS_Bombatoon_traversal_feet_air
 emitters [
  {
   emitterName=SplashElements001
   rendererClasses [&NiagaraMeshRendererProperties, &NiagaraSpriteRendererProperties]
  }
  {
   emitterName=Decal002
   rendererClasses [
    &NiagaraDecalRendererProperties
    &NiagaraSpriteRendererProperties
    &NiagaraMeshRendererProperties
   ]
  }
  {emitterName=Decal, rendererClasses [&NiagaraDecalRendererProperties, &NiagaraSpriteRendererProperties]}
 ]
 userVariables [
  User.bIsFrontEnd {type=NiagaraBool}
  User.bIsFrontEndPreview {type=NiagaraBool}
  User.bisHeadHidden {type=NiagaraBool}
  User.FrontEnd {type=NiagaraFloat, defaultValue=1, description=""}
  User.OwnerLocalVelocity {type=Vector3f, description="", defaultValue {struct=Vector3f, value=<0 0 0>}}
 ]
}
