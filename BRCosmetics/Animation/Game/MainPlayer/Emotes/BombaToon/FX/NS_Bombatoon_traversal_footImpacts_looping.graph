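# Niagara system for looping ink-like foot impact effects on the BombaToon traversal emote: decals, icon sprites, mesh splatter and stretching goop.
# Changes in user burst parameters trigger spawn groups at sampled skeletal mesh bones, alongside speed-gated continuous decal spawning.
# Owner velocity selects idle, traversal or rail/car tuning, with front-end overrides and time-of-day colors.
# Particles blend between local and world space, and goop traces to follow slopes.
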NiagaraSystem /BRCosmetics/Animation/Game/MainPlayer/Emotes/BombaToon/FX/NS_Bombatoon_traversal_footImpacts_looping
dependencies {
 usedDataInterfaces [
  &NiagaraDataInterfaceCurve
  &NiagaraDataInterfaceSpriteRendererInfo
  &NiagaraDataInterfaceCollisionQuery
  &NiagaraDataInterfaceDebugDraw
  &NiagaraDataInterfaceMeshRendererInfo
  &NiagaraDataInterfaceColorCurve
  &NiagaraDataInterfaceArrayFloat
  &NiagaraDataInterfaceSkeletalMesh
  &NiagaraDataInterfaceArrayPosition
  &NiagaraDataInterfaceArrayFloat3
 ]
 usedDynamicInputs []
 usedModules [
  &self:SystemSpawnScript.SystemScriptSource.SystemScriptGraph.NiagaraNodeAssignment_0.SetVariables_#1
  &self:SystemSpawnScript.SystemScriptSource.SystemScriptGraph.NiagaraNodeAssignment_5.SetVariables_#2
  &self:SystemSpawnScript.SystemScriptSource.SystemScriptGraph.NiagaraNodeAssignment_7.SetVariables_#3
  &self:SystemSpawnScript.SystemScriptSource.SystemScriptGraph.NiagaraNodeAssignment_8.SetVariables_#4
  &self:SystemSpawnScript.SystemScriptSource.SystemScriptGraph.NiagaraNodeAssignment_9.SetVariables_#5
  &SystemState
  &BlendSpaceHelper
  &self:ReadDeath
 ]
}
emitters {
 Decal002 {
  emitterSpawnScript [
   /SetVariables(Emitter.Idle Spawn Rate=10) {isSetParameters=true}
   /SetVariables(Emitter.Max Traversal speed=153, Emitter.Rail / Car Speed=500) {isSetParameters=true}
   /SetVariables(
    Emitter.Extra splatter Idle radius=20
    Emitter.Idle Area spawn:Vector3f=&RandomRangeVector(Minimum=<100 20 15>, Maximum=<-150 -150 15>)
    Emitter.Idle Spawn Radius=0.1
    Emitter.Rail/ Car Area spawn:Vector3f=&RandomRangeVector(Minimum=<35 65 15>, Maximum=<-35 -65 15>)
    Emitter.Traversal Spawn Radius=1
    Emitter.Traversal area Spawn:Vector3f=<0 -65 15>
    Emitter.rail/car Spawn Radius=35
   ) {isSetParameters=true}
   /SetVariables(
    Emitter.Layer 1 color:LinearColor=&FNTimeOfDayColor(Night Color=<rgba 0 0.0003 0.0015 1>, Morning Color=<rgba 0 0 0 1>, Day Color=<rgba 0 0 0 1>, Evening Color=<rgba 0 0 0 1>)
    Emitter.Layer 2 Color:LinearColor=&FNTimeOfDayColor(Night Color=<rgba 0.209 0.2945 0.38 1>, Morning Color=<rgba 1 1 1 1>, Day Color=<rgba 1 1 1 1>, Evening Color=<rgba 1 1 1 1>)
    Emitter.Layer 3 Color:LinearColor=&FNTimeOfDayColor(Night Color=<rgba 0 0.0003 0.0015 1>, Morning Color=<rgba 0 0 0 1>, Day Color=<rgba 0 0 0 1>, Evening Color=<rgba 0 0 0 1>)
    Emitter.Layer 4 Color:LinearColor=&FNTimeOfDayColor(Night Color=<rgba 0.1805 0.2779 0.38 1>, Morning Color=<rgba 1 1 1 1>, Day Color=<rgba 1 1 1 1>, Evening Color=<rgba 1 1 1 1>)
   ) {isSetParameters=true}
   /SetVariables(
    Emitter.Heart Traversal Lifespawn:NiagaraFloat=&RandomRangeFloat(Minimum=1, Maximum=1.3)
    Emitter.Icon on rail Lifespawn:NiagaraFloat=&RandomRangeFloat(Minimum=0.4, Maximum=0.6)
    "Emitter.Meow  Traversal Lifespawn":NiagaraFloat=&RandomRangeFloat(Minimum=1, Maximum=1.3)
    Emitter.Star Traversal Lifespawn:NiagaraFloat=&RandomRangeFloat(Minimum=0.5, Maximum=0.8)
   ) {isSetParameters=true}
  ]
  emitterUpdateScript [
   &EmitterState()
   /Game/Effects/Niagara/Modules/Logic/UpdateValueStateChange(InputIntParam:NiagaraInt32=&MakeIntFromFloat(Float=$User.BurstIndex, Conversion Mode=::NewEnumerator1))
   &self:TriggerSpawn_02(
    Spawn Count:NiagaraInt32=&MergeHelper_IntSwitch(
     0=1
     1=1
     2=1
     3=1
     4=1
     5=1
     6=1
     7=1
     8=1
     Index=&MakeIntFromFloat(Float=$User.BurstIndex, Conversion Mode=::NewEnumerator1)
     Num=::NewEnumerator7
    )
    Spawn Group:NiagaraInt32=&MakeIntFromFloat(Float=2, Conversion Mode=::NewEnumerator1)
    Trigger:NiagaraBool=$Emitter.UpdateValueStateChange.Changed
   ) {module=TriggerSpawn_02}
   &UpdateValueStateChange(InputIntParam:NiagaraInt32=&MakeIntFromFloat(Float=$User.KickbackBurst, Conversion Mode=::NewEnumerator1)) {module=UpdateValueStateChange001}
   &self:TriggerSpawn_01(
    Spawn Count:NiagaraInt32=&MergeHelper_IntSwitch(
     0=1
     1=1
     2=1
     3=1
     4=1
     5=1
     6=1
     7=1
     Index=&MakeIntFromFloat(Float=$User.KickbackBurst, Conversion Mode=::NewEnumerator1)
     Num=::NewEnumerator6
    )
    Spawn Group:NiagaraInt32=&MakeIntFromFloat(Float=3, Conversion Mode=::NewEnumerator1)
    Trigger:NiagaraBool=$Emitter.UpdateValueStateChange.Changed
   ) {module=TriggerSpawn_01}
   &SpawnRate(
    SpawnGroup=1
    SpawnRate=&FN_FrontEndFloatModifier(
     Gameplay=&FloatFromCurve(
      FloatCurve=DataInterface() {propertyValues {Curve {keys [<at 0 0 RCIM_Constant>,<at 0.15 1 RCIM_Constant>,<at 1 1>]}}}
      CurveIndex=$Emitter.NormalizedLoopAge
      Scale Curve=&MakeCustomFloatFromBool(
       Bool=&SetBoolByFloatComparison(
        A=&VectorLength(Vector=$User.OwnerLocalVelocity)
        B=$Emitter.Max Traversal speed
        Comparison Type=::NewEnumerator6
       )
       True Float=$Emitter.Idle Spawn Rate
       False Float=&MakeCustomFloatFromBool(
        Bool=&SetBoolByFloatComparison(
         A=&VectorLength(Vector=$User.OwnerLocalVelocity)
         B=Linked() {linkedVariable {name=Emitter.Rail / Car Speed, type=NiagaraFloat}}
         Comparison Type=::NewEnumerator6
        )
        True Float=$Emitter.Traversal Particle Spawn Rate
        False Float=Linked() {linkedVariable {name=Emitter.Rail/ Car Spawn Rate, type=NiagaraFloat}}
       )
      )
     )
     Front End=0
    )
    Use Spawn Probability=false
   ) {module=SpawnRate001}
   &SpawnRate(SpawnRate=&FN_FrontEndFloatModifier(Gameplay=20, Front End=0), Use Spawn Probability=false)
  ]
  particleSpawnScript [
   /SetVariables(Particles.VisibilityTag=&SpawnGroup()) {isSetParameters=true}
   &PartitionParticles()
   &InitializeParticle(
    Color=&FNTimeOfDayColor(Night Color=<rgba 0 0.0009 0.0156 1>, Morning Color=<rgba 0 0 0 1>, Day Color=<rgba 0 0 0 1>, Evening Color=<rgba 0 0 0 1>)
    Lifetime=&MergeHelper_FloatSwitch(
     0=&RandomRangeFloat(Minimum=0.85, Maximum=1)
     1=&RandomRangeFloat(Minimum=0.5, Maximum=0.7)
     2=&FN_FrontEndFloatModifier(
      Gameplay=&MakeCustomFloatFromBool(
       Bool=&SetBoolByFloatComparison(A=&VectorLength(Vector=$User.OwnerLocalVelocity), B=139, Comparison Type=::NewEnumerator6)
       True Float=$Emitter.Heart Traversal Lifespawn
       False Float=&MakeCustomFloatFromBool(
        Bool=&SetBoolByFloatComparison(A=&VectorLength(Vector=$User.OwnerLocalVelocity), B=500, Comparison Type=::NewEnumerator6)
        True Float=$Emitter.Heart Traversal Lifespawn
        False Float=$Emitter.Icon on rail Lifespawn
       )
      )
      Front End=$Emitter.Heart Traversal Lifespawn
     )
     3=&FN_FrontEndFloatModifier(
      Gameplay=&MakeCustomFloatFromBool(
       Bool=&SetBoolByFloatComparison(A=&VectorLength(Vector=$User.OwnerLocalVelocity), B=139, Comparison Type=::NewEnumerator6)
       True Float=Linked() {linkedVariable {name="Emitter.Meow  Traversal Lifespawn", type=NiagaraFloat}}
       False Float=&MakeCustomFloatFromBool(
        Bool=&SetBoolByFloatComparison(A=&VectorLength(Vector=$User.OwnerLocalVelocity), B=500, Comparison Type=::NewEnumerator6)
        True Float=Linked() {linkedVariable {name="Emitter.Meow  Traversal Lifespawn", type=NiagaraFloat}}
        False Float=$Emitter.Icon on rail Lifespawn
       )
      )
      Front End=Linked() {linkedVariable {name="Emitter.Meow  Traversal Lifespawn", type=NiagaraFloat}}
     )
     Index=&SpawnGroup()
     Num=::NewEnumerator2
    )
    Mass Max=1.5
    Mass Mode=::NewEnumerator3
    Sprite Rotation Mode=::NewEnumerator2
    Sprite Size=<50 50>
    Sprite Size Mode=::NewEnumerator4
    UsePositionOffset=false
    Write Material Random=true
   )
   &SkeletalMeshLocation(
    Bone / Socket Index=$Particles.PartitionParticles.Partition
    Bone Sampling Mode=::NewEnumerator4
    MaskSpawnGroup=false
    Sampled Position Offset Coordinate Space:ENiagaraSkelMeshTransforms=::NewEnumerator2
    Skeletal Mesh=$System.Skeletal Mesh
   )
   &ShapeLocation(
    Disc Coverage=0.85
    MaskSpawnGroup=false
    Offset=&MergeHelper_VectorSwitch(0=<0 0 0>, 1=<0 100 0>, 2=?, 3=?, Index=&SpawnGroup(), Num=::NewEnumerator0)
    Offset Coordinate Space=::Local
    Offset Mode=::NewEnumerator0
    Override Seed=false
    Ring Radius=&MergeHelper_FloatSwitch(
     0=&MakeCustomFloatFromBool(
      Bool=&SetBoolByFloatComparison(
       A=&VectorLength(Vector=$User.OwnerLocalVelocity)
       B=$Emitter.Max Traversal speed
       Comparison Type=::NewEnumerator6
      )
      True Float=$Emitter.Idle Spawn Radius
      False Float=&MakeCustomFloatFromBool(
       Bool=&SetBoolByFloatComparison(
        A=&VectorLength(Vector=$User.OwnerLocalVelocity)
        B=Linked() {linkedVariable {name=Emitter.Rail / Car Speed, type=NiagaraFloat}}
        Comparison Type=::NewEnumerator6
       )
       True Float=$Emitter.Traversal Spawn Radius
       False Float=Linked() {linkedVariable {name=Emitter.rail/car Spawn Radius, type=NiagaraFloat}}
      )
     )
     1=35
     2=3
     3=3
     Index=&SpawnGroup()
     Num=::NewEnumerator2
    )
    Shape Primitive=::NewEnumerator4
    UseOwnerScale=true
   )
  ]
  particleUpdateScript [
   &ParticleState()
   &ScaleColor(
    Scale Alpha=&FloatFromCurve(
     FloatCurve=DataInterface() {propertyValues {Curve {
        defaultValue=0
        keys [
         {interpMode=RCIM_Cubic, tangentMode=RCTM_User, time=0, value=0}
         {arriveTangent=0.2521, interpMode=RCIM_Cubic, leaveTangent=0.2521, tangentMode=RCTM_User, time=0.1147, value=1}
         {interpMode=RCIM_Cubic, tangentMode=RCTM_User, time=0.5, value=1}
         {arriveTangent=0.1836, interpMode=RCIM_Cubic, leaveTangent=0.1836, tangentMode=RCTM_User, time=0.7609, value=0.9806}
         {tangentMode=RCTM_User, time=1, value=0}
        ]
       }}}
     Scale Curve=3
    )
    Scale RGB=&MakeVectorFromLinearColorRGB(Color=<rgba 0 0 0 1>)
    ScaleA=true
    ScaleRGB=true
   )
   /SetVariables(Particles.DecalSize2:Vector3f=&Multiply_Vector(
     A=&VectorFromFloat(Value=&RandomRangeFloat(Minimum=10, Maximum=15))
     B=&VectorFromFloat(Value=&FloatFromCurve(
       FloatCurve=DataInterface() {propertyValues {Curve {keys [
           {arriveTangent=4.1785, interpMode=RCIM_Cubic, leaveTangent=4.1785, tangentMode=RCTM_User, time=0, value=0.5}
           {arriveTangent=11.9214, interpMode=RCIM_Cubic, leaveTangent=11.9214, tangentMode=RCTM_User, time=0.01, value=0.8}
           {arriveTangent=0.0224, interpMode=RCIM_Cubic, leaveTangent=0.0224, tangentMode=RCTM_User, time=0.05, value=1}
          ]}}}
       Scale Curve=&RandomRangeFloat(Minimum=2, Maximum=3.5)
      ))
    )) {isSetParameters=true}
   /SetVariables(Particles.DecalFade:NiagaraFloat=&FloatFromCurve(FloatCurve=DataInterface() {propertyValues {Curve {keys [<at 0 1>,<at 1 0>]}}})) {isSetParameters=true}
   /SetVariables(Particles.SpriteSize=&Multiply_Vector2D_ByFloat(
     Float=&FloatFromCurve(FloatCurve=DataInterface() {propertyValues {Curve {keys [
          <at 0 0 RCIM_Cubic>
          {arriveTangent=-0.0027, interpMode=RCIM_Cubic, leaveTangent=-0.0027, tangentMode=RCTM_User, time=0.0604, value=1.3}
          {arriveTangent=-0.4266, interpMode=RCIM_Cubic, leaveTangent=-0.4266, time=0.1204, value=1}
          {arriveTangent=-0.2399, interpMode=RCIM_Cubic, leaveTangent=-0.2399, tangentMode=RCTM_User, time=0.7636, value=1}
          {arriveTangent=-9.7345, interpMode=RCIM_Cubic, leaveTangent=-9.7345, tangentMode=RCTM_User, time=0.8449, value=1.5}
          <at 1 0 RCIM_Cubic>
         ]}}})
     Vector2D=<20 20>
    )) {isSetParameters=true}
   &self:ScratchModule(Index:NiagaraInt32=$Particles.PartitionParticles.Partition, Vector Array:NiagaraDataInterfaceArrayFloat3=$System.DeathPosition) {module=ScratchModule}
   /SetVariables(
    Particles. DynParam 0:Vector4f=&MakeVector4(X=0, Y=0, Z=1, W=1)
    Particles. DynParam 1:Vector4f=&MakeVector4(X=0, Y=0, Z=1, W=1)
   ) {isSetParameters=true}
   /SetVariables(
    Particles.Layer 1 DynParam 0:Vector4f=&MakeVector4(X=1, Y=1, Z=0, W=0)
    Particles.Layer 1 DynParam 1:Vector4f=&MakeVector4(X=1, Y=1, Z=0, W=0)
   ) {isSetParameters=true}
   &SubUVAnimation(
    End Frame Range Override=11
    Output Alternate Renderer Binding=true
    "Sprite Renderer "=DataInterface() {propertyValues {SpriteRenderer=&self:Decal002_1.NiagaraSpriteRendererProperties_5}}
    SubUV Lookup Index:NiagaraFloat=&FloatFromCurve(FloatCurve=DataInterface() {propertyValues {Curve {defaultValue=1, keys [<at 0 0>,<at 0.5 1>,<at 1 0>]}}})
    SubUV Lookup Index Scale:NiagaraFloat=&RandomRangeFloat(Minimum=0.8, Maximum=1)
    UseEndFrame=true
    UseStartFrame=true
   ) {module=SubUVAnimation002}
   &GravityForce(
    Coordinate Space=::Local
    Gravity=&MergeHelper_VectorSwitch(
     0=<0 0 0>
     1=<0 0 0>
     2=&MakeVector(
      X=&Lerp_Float(A=-250, B=250, Alpha=$User.KickbackBurst)
      Y=&FloatFromCurve(FloatCurve=DataInterface() {propertyValues {Curve {keys [<at 0 -350>,<at 0.5 -190>]}}})
      Z=&FloatFromCurve(FloatCurve=DataInterface() {propertyValues {Curve {keys [<at 0 0>,<at 0.2 950>,<at 0.4976 -450>,<at 0.75 650>,<at 1 -250>]}}})
     )
     3=&MakeVector(
      X=&Lerp_Float(A=200, B=-200, Alpha=$User.KickForward_Idle)
      Y=&FloatFromCurve(FloatCurve=DataInterface() {propertyValues {Curve {keys [<at 0 -350>,<at 0.5 -50>]}}}, Scale Curve=0.5)
      Z=&FloatFromCurve(FloatCurve=DataInterface() {propertyValues {Curve {keys [<at 0 0>,<at 0.1192 950>,<at 0.75 650>,<at 1 -250>]}}}, Scale Curve=0.45)
     )
     4=<0 0 0>
     Index=&SpawnGroup()
     Num=::NewEnumerator3
    )
   )
   &Drag(
    Drag=&MergeHelper_FloatSwitch(0=0, 1=0, 2=3, 3=3, 4=0, Index=&SpawnGroup(), Num=::NewEnumerator3)
    Use Linear Drag=true
    Use Rotational Drag=true
   )
   &SolveForcesAndVelocity(Clamp Velocity=false, Limit Acceleration=false)
   &SubUVAnimation(
    End Frame Range Override=10
    Output Alternate Renderer Binding=true
    "Sprite Renderer "=DataInterface() {propertyValues {SpriteRenderer=&self:Decal002_1.NiagaraSpriteRendererProperties_1}}
    SubUV Lookup Index:NiagaraFloat=&FloatFromCurve(FloatCurve=DataInterface() {propertyValues {Curve {defaultValue=1, keys [<at 0 0>,<at 0.5 1>,<at 1 0>]}}})
    SubUV Lookup Index Scale:NiagaraFloat=&RandomRangeFloat(Minimum=0.8, Maximum=1)
    UseEndFrame=true
    UseStartFrame=true
   )
   &BlendSpace("Position Blend (World <-> Local)":NiagaraFloat=&MergeHelper_FloatSwitch(
     0=0
     1=0
     2=&MakeCustomFloatFromBool(
      Bool=&SetBoolByFloatComparison(A=&VectorLength(Vector=$Engine.Owner.Velocity), B=400)
      True Float=1
      False Float=&FloatFromCurve(FloatCurve=DataInterface() {propertyValues {Curve {keys [<at 0 1 RCIM_Cubic>, {arriveTangent=-12.1775, interpMode=RCIM_Cubic, leaveTangent=-12.1775, tangentMode=RCTM_User, time=0.25, value=0}]}}})
     )
     3=&MakeCustomFloatFromBool(
      Bool=&SetBoolByFloatComparison(A=&VectorLength(Vector=$Engine.Owner.Velocity), B=400)
      True Float=1
      False Float=&FloatFromCurve(FloatCurve=DataInterface() {propertyValues {Curve {keys [<at 0 1 RCIM_Cubic>, {arriveTangent=-12.1775, interpMode=RCIM_Cubic, leaveTangent=-12.1775, tangentMode=RCTM_User, time=0.25, value=0}]}}})
     )
     Index=$Particles.VisibilityTag
     Num=::NewEnumerator2
     Translation Contribution:NiagaraFloat=1
     Rotation Contribution:NiagaraFloat=1
    ))
   /Niagara/Modules/Decal/Decal_Attributes(Decal Facing Mode=1, Fade=1, Orientation=1, Set Alignment=false, Size=1, Sort Order=1, Visible=1)
  ]
  rendererClasses [&NiagaraDecalRendererProperties, &NiagaraSpriteRendererProperties]
  renderers [
   {rendererClass=&NiagaraDecalRendererProperties, rendererIndex=0}
   {rendererClass=&NiagaraDecalRendererProperties, rendererIndex=1}
   {rendererClass=&NiagaraDecalRendererProperties, rendererIndex=2}
   {rendererClass=&NiagaraDecalRendererProperties, rendererIndex=3}
   {rendererClass=&NiagaraSpriteRendererProperties, rendererIndex=4}
   {rendererClass=&NiagaraSpriteRendererProperties, rendererIndex=5}
  ]
 }
 Foot_FX_merged {
  emitterSpawnScript [
   /SetVariables(
    Emitter.Car/rail Lifespawn Kick Forward:NiagaraFloat=&RandomRangeFloat(Minimum=0.25, Maximum=0.45)
    Emitter.Idle Lifespan Kick Forward:NiagaraFloat=&RandomRangeFloat(Minimum=0.35, Maximum=0.75)
   ) {isSetParameters=true}
   /SetVariables(Emitter.Max Traversal speed=153, Emitter.Rail / Car Speed=500) {isSetParameters=true}
   /SetVariables(
    "Emitter.Car/rail Lifespawn  Omni Foot Splatter":NiagaraFloat=&RandomRangeFloat(Minimum=0.15, Maximum=0.5)
    "Emitter.FE Lifespan  Omni Foot Splatter":NiagaraFloat=&RandomRangeFloat(Minimum=0.4, Maximum=0.65)
    "Emitter.Traversal LifeSpawn  Omni Foot Splatter":NiagaraFloat=&RandomRangeFloat(Minimum=0.2, Maximum=0.5)
   ) {isSetParameters=true}
   /SetVariables(
    Emitter.Car/rail Lifespawn Kick Back:NiagaraFloat=&RandomRangeFloat(Minimum=0.2, Maximum=0.5)
    Emitter.FE Lifespawn Kick Back:NiagaraFloat=&RandomRangeFloat(Minimum=0.2, Maximum=0.8)
    Emitter.Traversal LifeSpawn Kick Back:NiagaraFloat=&RandomRangeFloat(Minimum=0.2, Maximum=0.8)
   ) {isSetParameters=true}
   /SetVariables(
    Emitter.Car/rail Lifespawn Kick Forward:NiagaraFloat=&RandomRangeFloat(Minimum=0.25, Maximum=0.45)
    Emitter.Idle Lifespan Kick Forward:NiagaraFloat=&RandomRangeFloat(Minimum=0.35, Maximum=0.75)
   ) {isSetParameters=true}
  ]
  emitterUpdateScript [
   &EmitterState()
   &UpdateValueStateChange(InputIntParam:NiagaraInt32=&MakeIntFromFloat(Float=$User.KickForward_Idle, Conversion Mode=::NewEnumerator1))
   &self:TriggerSpawn_15(
    Spawn Count:NiagaraInt32=&MergeHelper_IntSwitch(
     0=11
     1=11
     2=11
     3=11
     4=11
     5=11
     6=11
     7=11
     Index=&MakeIntFromFloat(Float=$User.KickForward_Idle, Conversion Mode=::NewEnumerator1)
     Num=::NewEnumerator6
    )
    Spawn Group:NiagaraInt32=&MakeIntFromFloat(Float=2, Conversion Mode=::NewEnumerator1)
    Trigger:NiagaraBool=$Emitter.UpdateValueStateChange.Changed
   ) {module=TriggerSpawn_15}
   &UpdateValueStateChange(InputIntParam:NiagaraInt32=&MakeIntFromFloat(Float=$User.splashBurst, Conversion Mode=::NewEnumerator1)) {module=UpdateValueStateChange001}
   &self:TriggerSpawn_15(
    Spawn Count:NiagaraInt32=&MergeHelper_IntSwitch(
     0=15
     1=15
     2=15
     3=15
     4=15
     5=15
     6=15
     7=15
     Index=&MakeIntFromFloat(Float=$User.splashBurst, Conversion Mode=::NewEnumerator1)
     Num=::NewEnumerator6
    )
    Spawn Group:NiagaraInt32=&MakeIntFromFloat(Float=0, Conversion Mode=::NewEnumerator1)
    Trigger:NiagaraBool=$Emitter.UpdateValueStateChange001.Changed
   ) {module=TriggerSpawn001}
   &UpdateValueStateChange(InputIntParam:NiagaraInt32=&MakeIntFromFloat(Float=$User.KickbackBurst, Conversion Mode=::NewEnumerator1)) {module=UpdateValueStateChange002}
   &self:TriggerSpawn_15(
    Spawn Count:NiagaraInt32=&MergeHelper_IntSwitch(
     0=15
     1=15
     2=15
     3=15
     4=15
     5=15
     6=15
     7=15
     Index=&MakeIntFromFloat(Float=$User.KickbackBurst, Conversion Mode=::NewEnumerator1)
     Num=::NewEnumerator6
    )
    Spawn Group:NiagaraInt32=&MakeIntFromFloat(Float=1, Conversion Mode=::NewEnumerator1)
    Trigger:NiagaraBool=$Emitter.UpdateValueStateChange002.Changed
   ) {module=TriggerSpawn002}
  ]
  particleSpawnScript [
   /SetVariables(Particles.VisibilityTag=&SpawnGroup()) {isSetParameters=true}
   &FN_SetSpawnGroup(Number of Spawn Groups=3)
   &InitializeParticle(
    Color=&FNTimeOfDayColor(Night Color=<rgba 0 0.0009 0.0156 1>, Morning Color=<rgba 0 0 0 1>, Day Color=<rgba 0 0 0 1>, Evening Color=<rgba 0 0 0 1>)
    Lifetime=&MergeHelper_FloatSwitch(
     0=&FN_FrontEndFloatModifier(
      Gameplay=&MakeCustomFloatFromBool(
       Bool=&SetBoolByFloatComparison(A=&VectorLength(Vector=$User.OwnerLocalVelocity), B=153, Comparison Type=::NewEnumerator6)
       True Float=$Emitter.Idle Lifespan Omni Foot Splatter
       False Float=&MakeCustomFloatFromBool(
        Bool=&SetBoolByFloatComparison(A=&VectorLength(Vector=$User.OwnerLocalVelocity), B=500, Comparison Type=::NewEnumerator6)
        True Float=Linked() {linkedVariable {name="Emitter.Traversal LifeSpawn  Omni Foot Splatter", type=NiagaraFloat}}
        False Float=Linked() {linkedVariable {name="Emitter.Car/rail Lifespawn  Omni Foot Splatter", type=NiagaraFloat}}
       )
      )
      Front End=Linked() {linkedVariable {name="Emitter.FE Lifespan  Omni Foot Splatter", type=NiagaraFloat}}
     )
     1=&FN_FrontEndFloatModifier(
      Gameplay=&MakeCustomFloatFromBool(
       Bool=&SetBoolByFloatComparison(A=&VectorLength(Vector=$User.OwnerLocalVelocity), B=153, Comparison Type=::NewEnumerator6)
       True Float=$Emitter.Idle Lifespan Kick back
       False Float=&MakeCustomFloatFromBool(
        Bool=&SetBoolByFloatComparison(A=&VectorLength(Vector=$User.OwnerLocalVelocity), B=500, Comparison Type=::NewEnumerator6)
        True Float=$Emitter.Traversal LifeSpawn Kick Back
        False Float=Linked() {linkedVariable {name=Emitter.Car/rail Lifespawn Kick Back, type=NiagaraFloat}}
       )
      )
      Front End=$Emitter.FE Lifespawn Kick Back
     )
     2=&FN_FrontEndFloatModifier(
      Gameplay=&MakeCustomFloatFromBool(
       Bool=&SetBoolByFloatComparison(
        A=&VectorLength(Vector=$User.OwnerLocalVelocity)
        B=$Emitter.Max Traversal speed
        Comparison Type=::NewEnumerator6
       )
       True Float=$Emitter.Idle Lifespan Kick Forward
       False Float=&MakeCustomFloatFromBool(
        Bool=&SetBoolByFloatComparison(
         A=&VectorLength(Vector=$User.OwnerLocalVelocity)
         B=Linked() {linkedVariable {name=Emitter.Rail / Car Speed, type=NiagaraFloat}}
         Comparison Type=::NewEnumerator6
        )
        True Float=$Emitter.Traversal LifeSpawn Kick Forward
        False Float=Linked() {linkedVariable {name=Emitter.Car/rail Lifespawn Kick Forward, type=NiagaraFloat}}
       )
      )
      Front End=0
     )
     3=?
     Index=&SpawnGroup()
     Num=::NewEnumerator1
    )
    Mesh Scale=&MergeHelper_VectorSwitch(
     0=&Multiply_VectorByFloat(Vector=<1.5 1 1>, Float=&RandomRangeFloat(Minimum=0.5, Maximum=1.5))
     1=&Multiply_VectorByFloat(Vector=<1.5 1 1>, Float=&RandomRangeFloat(Minimum=0.5, Maximum=1.5))
     2=&Multiply_VectorByFloat(Vector=<1.5 1 1>, Float=&RandomRangeFloat(Minimum=0.5, Maximum=1.5))
     3=<5 5 5>
     4=<0 0 0>
     Index=&SpawnGroup()
     Num=::NewEnumerator3
    )
    Mesh Scale Mode=::NewEnumerator4
    Sprite Size Mode=::NewEnumerator3
    Uniform Sprite Size=&MergeHelper_FloatSwitch(
     0=&MakeFloatFromVector(
      Vector=&Multiply_VectorByFloat(Vector=<2 1 1>, Float=&RandomRangeFloat(Minimum=0.5, Maximum=1.7))
      Channel=::NewEnumerator0
     )
     1=50
     2=&RandomRangeFloat(Minimum=10, Maximum=50)
     3=0
     4=0
     Index=&SpawnGroup()
     Num=::NewEnumerator3
    )
    UsePositionOffset=false
    Write Material Random=true
   )
   &AddVelocity(
    Cone Angle=&MergeHelper_FloatSwitch(0=140.6455, 1=20, 2=0, 3=0, 4=0, Index=&SpawnGroup(), Num=::NewEnumerator3)
    Cone Axis=&MergeHelper_VectorSwitch(
     0=<0 0 1>
     1=&RandomRangeVector(Minimum=<0 0 0.5>, Maximum=<0 -1 1>)
     2=<0 0 0>
     3=<0 0 0>
     4=<0 0 0>
     Index=&SpawnGroup()
     Num=::NewEnumerator3
    )
    Module.UseConeFalloff=false
    Module.UseVelDistribution=true
    Override Seed=false
    Velocity Mode=::NewEnumerator2
    Velocity Speed=&MergeHelper_FloatSwitch(
     0=&RandomRangeFloat(Minimum=50, Maximum=150)
     1=&RandomRangeFloat(Minimum=15, Maximum=250)
     2=0
     3=?
     Index=&SpawnGroup()
     Num=::NewEnumerator1
    )
   )
   &ApplyInitialForces()
   &SubUVAnimation(
    End Frame Range Override=3
    Mesh Renderer=DataInterface() {propertyValues {MeshRenderer=&self:Foot_FX_merged_0.NiagaraMeshRendererProperties_0}}
    Output Alternate Renderer Binding=true
    Renderer Type:ENiagaraMeshOrSprite=::NewEnumerator1
    SubUV Animation Mode:ENiagara_SubUVLookupModeV2=::NewEnumerator2
    UseEndFrame=true
    UseStartFrame=true
   )
   &SkeletalMeshLocation(
    Bone / Socket Index=&MakeIntFromFloat(Float=$User.splashBurst, Conversion Mode=::NewEnumerator1)
    Bone Sampling Mode=::NewEnumerator4
    MaskSpawnGroup=false
    Orientation Sampling=::NewEnumerator1
    Position Sampling=::NewEnumerator2
    Sampled Position Offset=&MergeHelper_VectorSwitch(
     0=<0 -7 -7>
     1=<0 -7 -7>
     2=?
     3=?
     Index=&MakeIntFromFloat(Float=$User.splashBurst, Conversion Mode=::NewEnumerator1)
     Num=::NewEnumerator0
    )
    Sampled Position Offset Coordinate Space:ENiagaraSkelMeshTransforms=::NewEnumerator2
    Skeletal Mesh=$System.Skeletal Mesh
   )
   &SkeletalMeshLocation(
    Bone / Socket Index=&MakeIntFromFloat(Float=$User.KickbackBurst, Conversion Mode=::NewEnumerator1)
    Bone Sampling Mode=::NewEnumerator4
    MaskSpawnGroup=false
    Orientation Sampling=::NewEnumerator1
    Position Sampling=::NewEnumerator2
    Sampled Position Offset=&MergeHelper_VectorSwitch(
     0=<0 -5 0>
     1=<0 -5 0>
     2=?
     3=?
     Index=&MakeIntFromFloat(Float=$User.KickbackBurst, Conversion Mode=::NewEnumerator1)
     Num=::NewEnumerator0
    )
    Sampled Position Offset Coordinate Space:ENiagaraSkelMeshTransforms=::NewEnumerator2
    Skeletal Mesh=$System.Skeletal Mesh
   ) {module=SkeletalMeshLocation001}
   &SkeletalMeshLocation(
    Bone / Socket Index=&MakeIntFromFloat(Float=$User.KickForward_Idle, Conversion Mode=::NewEnumerator1)
    Bone Sampling Mode=::NewEnumerator4
    MaskSpawnGroup=false
    Orientation Sampling=::NewEnumerator1
    Position Sampling=::NewEnumerator2
    Sampled Position Offset=&MergeHelper_VectorSwitch(
     0=<0 -5 0>
     1=<0 -5 0>
     2=?
     3=?
     Index=&MakeIntFromFloat(Float=$User.KickForward_Idle, Conversion Mode=::NewEnumerator1)
     Num=::NewEnumerator0
    )
    Skeletal Mesh=$System.Skeletal Mesh
   ) {module=SkeletalMeshLocation002}
   &CameraOffset(Camera Offset Amount=&MergeHelper_FloatSwitch(
     0=&RandomRangeFloat(Minimum=0, Maximum=0)
     1=&RandomRangeFloat(Minimum=-10, Maximum=10)
     2=&RandomRangeFloat(Minimum=-10, Maximum=10)
     3=&RandomRangeFloat(Minimum=-5, Maximum=5)
     4=&RandomRangeFloat(Minimum=-8, Maximum=8)
     Index=&SpawnGroup()
     Num=::NewEnumerator3
    ))
   &SubUVAnimation(
    End Frame Range Override=3
    Mesh Renderer=DataInterface() {propertyValues {MeshRenderer=&self:Foot_FX_merged_0.NiagaraMeshRendererProperties_1}}
    Output Alternate Renderer Binding=true
    Renderer Type:ENiagaraMeshOrSprite=::NewEnumerator1
    SubUV Animation Mode:ENiagara_SubUVLookupModeV2=::NewEnumerator2
    UseEndFrame=true
    UseStartFrame=true
   ) {module=SubUVAnimation001}
   &AddVelocity(Velocity=&MergeHelper_VectorSwitch(
     0=<0 0 0>
     1=<0 0 0>
     2=&RandomRangeVector(Minimum=<-50 -5 100>, Maximum=<50 150 200>)
     3=<0 0 0>
     4=<0 0 0>
     Index=&SpawnGroup()
     Num=::NewEnumerator3
    )) {module=AddVelocity001}
   /SetVariables(Particles.Kickforward Position:NiagaraPosition=$Particles.SkeletalMeshLocation002.SampledPosition) {isSetParameters=true}
   /SetVariables(Particles.Omni Position:NiagaraPosition=$Particles.SkeletalMeshLocation.SampledPosition) {isSetParameters=true}
   /SetVariables(Particles.Kickback Position:NiagaraPosition=$Particles.SkeletalMeshLocation001.SampledPosition) {isSetParameters=true}
   &ShapeLocation(
    MaskSpawnGroup=false
    Override Seed=false
    Sphere Radius=&MergeHelper_FloatSwitch(0=0, 1=1, 2=3, 3=3, Index=&SpawnGroup(), Num=::NewEnumerator2)
    UseOwnerScale=true
   )
  ]
  particleUpdateScript [
   &ParticleState()
   &ScaleSpriteSize(Uniform Curve Sprite Scale=DataInterface() {propertyValues {Curve {keys [
        <at 0 0 RCIM_Cubic>
        {arriveTangent=1, interpMode=RCIM_Cubic, leaveTangent=1, time=0.1, value=1}
        {arriveTangent=-1.1111, interpMode=RCIM_Cubic, leaveTangent=-1.1111, time=0.7, value=0.7}
        <at 1 0 RCIM_Cubic>
       ]}}})
   &ScaleMeshSize(Scale Factor=&MergeHelper_VectorSwitch(
     0=&VectorFromFloat(Value=&FloatFromCurve(
       FloatCurve=DataInterface() {propertyValues {Curve {keys [<at 0 1 RCIM_Cubic>, {arriveTangent=-1, interpMode=RCIM_Cubic, leaveTangent=-1, time=0.3, value=1.125}, <at 1 0 RCIM_Cubic>]}}}
       Scale Curve=&FN_FrontEndFloatModifier(Gameplay=1, Front End=0.8)
      ))
     1=&VectorFromFloat(Value=&FloatFromCurve(
       FloatCurve=DataInterface() {propertyValues {Curve {keys [<at 0 1 RCIM_Cubic>, {arriveTangent=-1, interpMode=RCIM_Cubic, leaveTangent=-1, time=0.3, value=1.125}, <at 1 0 RCIM_Cubic>]}}}
       Scale Curve=&FN_FrontEndFloatModifier(Gameplay=0.8, Front End=0.5)
      ))
     2=&VectorFromFloat(Value=&FloatFromCurve(
       FloatCurve=DataInterface() {propertyValues {Curve {keys [<at 0 1 RCIM_Cubic>, {arriveTangent=-1, interpMode=RCIM_Cubic, leaveTangent=-1, time=0.3, value=1.125}, <at 1 0 RCIM_Cubic>]}}}
       Scale Curve=&RandomRangeFloat(Minimum=0.5, Maximum=1)
      ))
     3=?
     Index=&SpawnGroup()
     Num=::NewEnumerator1
    ))
   &GravityForce(Gravity=&MergeHelper_VectorSwitch(
     0=&FN_FrontEndVector3Modifier(
      Gameplay=&RandomRangeVector(Minimum=<0 0 -350>, Maximum=<0 0 -556.8417>)
      Front End=&RandomRangeVector(Minimum=<0 0 -350>, Maximum=<0 0 -980>)
     )
     1=&RandomRangeVector(Minimum=<0 0 -556.8417>, Maximum=<1 1 -980>)
     2=&RandomRangeVector(Minimum=<0 0 -980>, Maximum=<0 0 -1000>)
     3=?
     Index=&SpawnGroup()
     Num=::NewEnumerator1
    ))
   /SetVariables(Particles.Omni splatter Color:LinearColor=&MakeLinearColorFromVectorAndFloat(
     "Vector (RGB)"=<1 1 1>
     "Float (Alpha)"=&FloatFromCurve(FloatCurve=DataInterface() {propertyValues {Curve {keys [<at -0.0046 1.0113 RCIM_Cubic>,<at 1.0011 0.002 RCIM_Cubic>]}}})
    )) {isSetParameters=true}
   /SetVariables(Particles.Kickforward Position:NiagaraPosition=&AddVectorToPosition(
     Position=$Particles.Kickforward Position
     Vector=&Multiply_Vector(A=$Particles.Velocity, B=&VectorFromFloat(Value=$Engine.DeltaTime))
    )) {isSetParameters=true}
   /SetVariables(Particles.Kickback Position:NiagaraPosition=&AddVectorToPosition(
     Position=$Particles.Kickback Position
     Vector=&Multiply_Vector(A=$Particles.Velocity, B=&VectorFromFloat(Value=$Engine.DeltaTime))
    )) {isSetParameters=true}
   /SetVariables(Particles.Omni Position:NiagaraPosition=&AddVectorToPosition(
     Position=$Particles.Omni Position
     Vector=&Multiply_Vector(A=$Particles.Velocity, B=&VectorFromFloat(Value=$Engine.DeltaTime))
    )) {isSetParameters=true}
   &Drag(
    Drag=&MergeHelper_FloatSwitch(
     0=0.1091
     1=&FloatFromCurve(
      FloatCurve=DataInterface() {propertyValues {Curve {defaultValue=1, keys [{tangentMode=RCTM_User, time=0, value=0}, {tangentMode=RCTM_User, time=1, value=1}]}}}
      Scale Curve=&RandomRangeFloat(Minimum=0, Maximum=0.5)
     )
     2=&FloatFromCurve(
      FloatCurve=DataInterface() {propertyValues {Curve {defaultValue=1, keys [{tangentMode=RCTM_User, time=0, value=0}, {tangentMode=RCTM_User, time=1, value=1}]}}}
      Scale Curve=&RandomRangeFloat(Minimum=0, Maximum=2)
     )
     3=?
     Index=&SpawnGroup()
     Num=::NewEnumerator1
    )
    Use Linear Drag=true
    Use Rotational Drag=true
   )
   &SolveForcesAndVelocity(Clamp Velocity=false, Limit Acceleration=false)
   &ScaleMeshSizeBySpeed(Sample Scale Factor By Curve=true)
   /SetVariables(Particles.Kick back splatter color:LinearColor=&MakeLinearColorFromVectorAndFloat(
     "Vector (RGB)"=<1 1 1>
     "Float (Alpha)"=&FloatFromCurve(FloatCurve=DataInterface() {propertyValues {Curve {keys [<at -0.0046 1.0113 RCIM_Cubic>,<at 1.0011 0.002 RCIM_Cubic>]}}})
    )) {isSetParameters=true}
   /SetVariables(Particles.INK SpriteSize:Vector2f=&Multiply_Vector2D_ByFloat(
     Float=&RandomRangeFloat(Minimum=1, Maximum=3)
     Vector2D=&RandomRangeVector2D(Minimum=<1.9792 5>, Maximum=<3 5>)
    )) {isSetParameters=true}
   &BlendSpace("Position Blend (World <-> Local)":NiagaraFloat=&MakeCustomFloatFromBool(
     Bool=&SetBoolByFloatComparison(A=&VectorLength(Vector=$Engine.Owner.Velocity), B=400)
     True Float=1
     False Float=&FloatFromCurve(FloatCurve=DataInterface() {propertyValues {Curve {keys [<at 0 1 RCIM_Cubic>, {arriveTangent=-12.1775, interpMode=RCIM_Cubic, leaveTangent=-12.1775, tangentMode=RCTM_User, time=0.25, value=0}]}}})
     Translation Contribution:NiagaraFloat=1
     Rotation Contribution:NiagaraFloat=0
    )) {module=BlendSpace001}
  ]
  rendererClasses [&NiagaraMeshRendererProperties, &NiagaraSpriteRendererProperties]
  renderers [
   {rendererClass=&NiagaraMeshRendererProperties, rendererIndex=0}
   {rendererClass=&NiagaraMeshRendererProperties, rendererIndex=1}
   {rendererClass=&NiagaraMeshRendererProperties, rendererIndex=2}
   {rendererClass=&NiagaraSpriteRendererProperties, rendererIndex=3}
  ]
 }
 GOOP_STRETCH {
  emitterSpawnScript [
   /SetVariables(Emitter.Max Traversal speed=153, Emitter.Rail / Car Speed=154) {isSetParameters=true}
   /SetVariables(Emitter.Traversal LifeSpawn=0.5) {isSetParameters=true}
  ]
  emitterUpdateScript [
   &EmitterState()
   &UpdateValueStateChange(InputIntParam:NiagaraInt32=&MakeIntFromFloat(Float=$User.KickbackBurst, Conversion Mode=::NewEnumerator3))
   &self:TriggerSpawn_06(
    Spawn Count:NiagaraInt32=&MergeHelper_IntSwitch(
     0=1
     1=1
     2=1
     3=1
     4=1
     5=1
     6=1
     7=1
     8=1
     Index=&MakeIntFromFloat(Float=$User.BurstIndex, Conversion Mode=::NewEnumerator1)
     Num=::NewEnumerator7
    )
    Spawn Group:NiagaraInt32=&MakeIntFromFloat(Float=0, Conversion Mode=::NewEnumerator1)
    Trigger:NiagaraBool=$Emitter.UpdateValueStateChange.Changed
   ) {module=TriggerSpawn_06}
   &self:TriggerSpawn_07(
    Spawn Count:NiagaraInt32=&MergeHelper_IntSwitch(
     0=1
     1=1
     2=1
     3=1
     4=1
     5=1
     6=1
     7=1
     Index=&MakeIntFromFloat(Float=$User.KickbackBurst, Conversion Mode=::NewEnumerator1)
     Num=::NewEnumerator6
    )
    Spawn Group:NiagaraInt32=&MakeIntFromFloat(Float=0, Conversion Mode=::NewEnumerator1)
    Trigger:NiagaraBool=$Emitter.UpdateValueStateChange.Changed
   ) {module=TriggerSpawn_07}
  ]
  particleSpawnScript [
   &PartitionParticles(Number Of Partitions=&SpawnGroup()) {enabled=false}
   /SetVariables(Particles.VisibilityTag=&SpawnGroup()) {isSetParameters=true}
   &InitializeParticle(
    Color=&FNTimeOfDayColor(Night Color=<rgba 0 0.0009 0.0156 1>, Morning Color=<rgba 0 0 0 1>, Day Color=<rgba 0 0 0 1>, Evening Color=<rgba 0 0 0 1>)
    Lifetime=&FN_FrontEndFloatModifier(
     Gameplay=&MakeCustomFloatFromBool(
      Bool=&SetBoolByFloatComparison(A=&VectorLength(Vector=$User.OwnerLocalVelocity), B=139, Comparison Type=::NewEnumerator6)
      True Float=$Emitter.Idle Lifespan
      False Float=&MakeCustomFloatFromBool(
       Bool=&SetBoolByFloatComparison(A=&VectorLength(Vector=$User.OwnerLocalVelocity), B=500, Comparison Type=::NewEnumerator6)
       True Float=$Emitter.Traversal LifeSpawn
       False Float=Linked() {linkedVariable {name=Emitter.Car/rail Lifespawn, type=NiagaraFloat}}
      )
     )
     Front End=0
    )
    Mass Mode=::NewEnumerator3
    Mesh Scale=&Multiply_VectorByFloat(Vector=<1 1 1>, Float=&RandomRangeFloat(Minimum=20, Maximum=20))
    Mesh Scale Mode=::NewEnumerator4
    Position Offset=<0 -35 0>
    Sprite Size Mode=::NewEnumerator1
    Uniform Sprite Size Max=20
    UsePositionOffset=true
    Write Material Random=true
   )
   &SkeletalMeshLocation(
    Bone / Socket Index=&MakeIntFromFloat(Float=$User.KickbackBurst, Conversion Mode=::NewEnumerator1)
    Bone Sampling Mode=::NewEnumerator4
    MaskSpawnGroup=false
    Sampled Position Offset=&MergeHelper_VectorSwitch(
     0=<-65 10 0>
     1=<-15 10 0>
     2=?
     3=?
     Index=&MakeIntFromFloat(Float=$User.KickbackBurst, Conversion Mode=::NewEnumerator1)
     Num=::NewEnumerator0
    )
    Sampled Position Offset Coordinate Space:ENiagaraSkelMeshTransforms=::NewEnumerator1
    Skeletal Mesh=$System.Skeletal Mesh
   )
   &SubUVAnimation(End Frame=3, SubUV Animation Mode:ENiagara_SubUVLookupMode=::NewEnumerator2)
  ]
  particleUpdateScript [
   &ParticleState()
   &Drag(
    Drag=&FloatFromCurve(
     FloatCurve=DataInterface() {propertyValues {Curve {defaultValue=1, keys [{tangentMode=RCTM_User, time=0, value=0}, {tangentMode=RCTM_User, time=1, value=1}]}}}
     Scale Curve=&RandomRangeFloat(Minimum=0, Maximum=0.5)
    )
    Use Linear Drag=true
    Use Rotational Drag=false
   )
   &ScaleColor(
    Scale Alpha=&FloatFromCurve(FloatCurve=DataInterface() {propertyValues {Curve {keys [<at -0.0046 1.0113 RCIM_Cubic>,<at 1.0011 0.002 RCIM_Cubic>]}}})
    ScaleA=true
    ScaleRGB=true
   )
   &SolveForcesAndVelocity(Clamp Velocity=false, Limit Acceleration=false)
   &ScaleSpriteSize(Uniform Curve Sprite Scale=DataInterface() {propertyValues {Curve {keys [
        <at 0 0 RCIM_Cubic>
        {arriveTangent=1, interpMode=RCIM_Cubic, leaveTangent=1, time=0.1, value=1}
        {arriveTangent=-1.1111, interpMode=RCIM_Cubic, leaveTangent=-1.1111, time=0.7, value=0.7}
        <at 1 0 RCIM_Cubic>
       ]}}})
   &CameraOffset(Camera Offset Amount=&RandomRangeFloat(Minimum=-10, Maximum=10))
   &ScaleMeshSize(Scale Factor=&MakeVector(
     X=0.03
     Y=0.03
     Z=&FloatFromCurve(
      FloatCurve=DataInterface() {propertyValues {Curve {defaultValue=1, keys [{tangentMode=RCTM_User, time=0, value=0}, {tangentMode=RCTM_User, time=0.1, value=0.9806}, <at 0.5 0>]}}}
      Scale Curve=0.2
     )
    ))
   &DynamicMaterialParameters(
    Index 0 Param 1=10
    Index 0 Param 2=0.536
    Module.Index0Float0Write=true
    Module.Index0Float1Write=true
    Module.Index0Float2Write=true
    Module.Index0Float3Write=true
    Module.Index1Float0Write=true
    Module.Index1Float1Write=true
    Module.Index1Float2Write=true
    Module.Index1Float3Write=true
    Write Parameter Index 1=::NewEnumerator1
   )
   /SetVariables(Particles.Stringy Scale:Vector3f=&MakeVector(
     X=&FloatFromCurve(
      FloatCurve=DataInterface() {propertyValues {Curve {defaultValue=1, keys [{tangentMode=RCTM_User, time=0, value=0}, {tangentMode=RCTM_User, time=0.3876, value=0.9806}, <at 0.567 0>]}}}
      Scale Curve=4
     )
     Y=5
     Z=5
    )) {isSetParameters=true}
   &self:Slope_Position_15(
    Apply Position:NiagaraBool=&InvertBool(Boolean:NiagaraBool=$User.bIsFrontEndPreview)
    Apply Rotation:NiagaraBool=&InvertBool(Boolean:NiagaraBool=$User.bIsFrontEndPreview)
    Fallback Position:NiagaraPosition=&AddVectorToPosition(Position=$Particles.Initial.Position, Vector=<1 1 1>)
    Height Adjustment/Offset=0.5
    Slope Position:NiagaraPosition=$Particles.SkeletalMeshLocation.SampledPosition
    Trace Length=120
    Trace Origin Height=85
    bIsFrontEnd:NiagaraBool=$User.bIsFrontEndPreview
   ) {module=Slope_Position_15}
   /SetVariables(
    Particles.Layer 1 DynParam 0:Vector4f=&MakeVector4(
     X=&FloatFromCurve(FloatCurve=DataInterface() {propertyValues {Curve {keys [
          {interpMode=RCIM_Cubic, tangentMode=RCTM_Break, time=0.0356, value=0.9644}
          {interpMode=RCIM_Cubic, tangentMode=RCTM_Break, time=0.8467, value=0.8}
          {interpMode=RCIM_Cubic, tangentMode=RCTM_Break, time=1, value=0}
         ]}}})
     Y=&FloatFromCurve(FloatCurve=DataInterface() {propertyValues {Curve {keys [
          {interpMode=RCIM_Cubic, tangentMode=RCTM_Break, time=0.0356, value=0.9644}
          {interpMode=RCIM_Cubic, tangentMode=RCTM_Break, time=0.8467, value=0.8}
          {interpMode=RCIM_Cubic, tangentMode=RCTM_Break, time=1, value=0}
         ]}}})
     Z=0
     W=0
    )
    Particles.Layer 1 DynParam 1:Vector4f=&MakeVector4(X=1, Y=1, Z=0, W=0)
   ) {isSetParameters=true}
   /SetVariables(
    Particles.Fake Decal SpriteSize:Vector2f=&Multiply_Vector2D_ByFloat(
     Float=&FloatFromCurve(
      FloatCurve=DataInterface() {propertyValues {Curve {
         defaultValue=1
         keys [{interpMode=RCIM_Cubic, tangentMode=RCTM_Break, time=0, value=0}, {interpMode=RCIM_Cubic, tangentMode=RCTM_Break, time=0.1781, value=1}]
        }}}
      Scale Curve=&RandomRangeFloat(Minimum=1, Maximum=1.3)
     )
     Vector2D=<50 50>
    )
    Particles.Fake decal SpriteRotation:NiagaraFloat=&RandomRangeFloat(Minimum=-1, Maximum=1)
    Particles.Fake decal color:LinearColor=&MakeLinearColorFromVectorAndFloat("Vector (RGB)"=<1 1 1>, "Float (Alpha)"=1)
   ) {isSetParameters=true}
  ]
  rendererClasses [&NiagaraMeshRendererProperties, &NiagaraSpriteRendererProperties]
  renderers [
   {rendererClass=&NiagaraMeshRendererProperties, rendererIndex=0}
   {rendererClass=&NiagaraSpriteRendererProperties, rendererIndex=1}
   {rendererClass=&NiagaraSpriteRendererProperties, rendererIndex=2}
  ]
 }
 foot_impact_w_slope_detection001 {
  emitterSpawnScript [
   /SetVariables(Emitter.Max Traversal speed=153, Emitter.Rail / Car Speed=500) {isSetParameters=true}
   /SetVariables(Emitter.Idle Drag=3, Emitter.Rail drag=5, Emitter.Traversal Drag=5) {isSetParameters=true}
   /SetVariables(
    Emitter.Idle Lifespawn decal2:NiagaraFloat=&RandomRangeFloat(Minimum=0.5, Maximum=1)
    Emitter.Traversal LifeSpawn=0.4
    Emitter.Traversal lifespawn decal 2:NiagaraFloat=&RandomRangeFloat(Minimum=0.5, Maximum=1)
   ) {isSetParameters=true}
   /SetVariables(Emitter.fake decal color:LinearColor=&MakeLinearColorFromVectorAndFloat(
     "Vector (RGB)"=&MakeVectorFromLinearColorRGB(Color=&ColorFromCurve(
       ColorCurve=DataInterface() {propertyValues {
         AlphaCurve {keys [<at 0 1 RCIM_Cubic>,<at 0.6566 0 RCIM_Cubic>]}
         BlueCurve {keys [<at 0 0 RCIM_Cubic>,<at 1 0 RCIM_Cubic>]}
         GreenCurve {keys [<at 0 0 RCIM_Cubic>,<at 1 0 RCIM_Cubic>]}
         RedCurve {keys [<at 0 0 RCIM_Cubic>,<at 1 0 RCIM_Cubic>]}
        }}
       CurveIndex=$Emitter.NormalizedLoopAge
      ))
     "Float (Alpha)"=&FloatFromCurve(
      FloatCurve=DataInterface() {propertyValues {Curve {keys [
          {interpMode=RCIM_Cubic, tangentMode=RCTM_Break, time=0, value=1}
          {interpMode=RCIM_Cubic, tangentMode=RCTM_Break, time=0.816, value=0.7016}
          {interpMode=RCIM_Cubic, tangentMode=RCTM_Break, time=1, value=0}
         ]}}}
      CurveIndex=$Emitter.NormalizedLoopAge
     )
    )) {isSetParameters=true}
   /SetVariables(
    Emitter.Layer 1 color:LinearColor=&FNTimeOfDayColor(Night Color=<rgba 0 0.0003 0.0015 1>, Morning Color=<rgba 0 0 0 1>, Day Color=<rgba 0 0 0 1>, Evening Color=<rgba 0 0 0 1>)
    Emitter.Layer 2 Color:LinearColor=&FNTimeOfDayColor(
     Night Color=<rgba 0.1505 0.2311 0.35 1>
     Morning Color=<rgba 0.5104 0.4402 0.4354 1>
     Day Color=<rgba 1 1 1 1>
     Evening Color=<rgba 0.349 0.3009 0.2977 1>
    )
    Emitter.Layer 3 Color:LinearColor=&FNTimeOfDayColor(Night Color=<rgba 0 0.0003 0.0015 1>, Morning Color=<rgba 0 0 0 1>, Day Color=<rgba 0 0 0 1>, Evening Color=<rgba 0 0 0 1>)
    Emitter.Layer 4 Color:LinearColor=&FNTimeOfDayColor(
     Night Color=<rgba 0.1505 0.2311 0.35 1>
     Morning Color=<rgba 0.5104 0.4402 0.4354 1>
     Day Color=<rgba 1 1 1 1>
     Evening Color=<rgba 0.349 0.3009 0.2977 1>
    )
   ) {isSetParameters=true}
   /SetVariables(
    Emitter.Heart Traversal Lifespawn:NiagaraFloat=&RandomRangeFloat(Minimum=1, Maximum=1.3)
    Emitter.Icon on rail Lifespawn:NiagaraFloat=&RandomRangeFloat(Minimum=0, Maximum=0)
    "Emitter.Meow  Traversal Lifespawn":NiagaraFloat=&RandomRangeFloat(Minimum=1, Maximum=1.3)
    Emitter.Star Traversal Lifespawn:NiagaraFloat=&RandomRangeFloat(Minimum=0.5, Maximum=0.8)
   ) {isSetParameters=true}
   /SetVariables(Emitter.Idle Spawn Rate=10) {isSetParameters=true}
  ]
  emitterUpdateScript [
   &EmitterState()
   &SpawnRate(
    SpawnGroup=6
    SpawnRate=&FN_FrontEndFloatModifier(
     Gameplay=&FloatFromCurve(
      FloatCurve=DataInterface() {propertyValues {Curve {keys [<at 0 0 RCIM_Constant>,<at 0.0977 0.9647 RCIM_Constant>,<at 1 1>]}}}
      CurveIndex=$Emitter.NormalizedLoopAge
      Scale Curve=&MakeCustomFloatFromBool(
       Bool=&SetBoolByFloatComparison(
        A=&VectorLength(Vector=$User.OwnerLocalVelocity)
        B=$Emitter.Max Traversal speed
        Comparison Type=::NewEnumerator6
       )
       True Float=$Emitter.Idle Spawn Rate
       False Float=&MakeCustomFloatFromBool(
        Bool=&SetBoolByFloatComparison(
         A=&VectorLength(Vector=$User.OwnerLocalVelocity)
         B=Linked() {linkedVariable {name=Emitter.Rail / Car Speed, type=NiagaraFloat}}
         Comparison Type=::NewEnumerator6
        )
        True Float=0
        False Float=0
       )
      )
     )
     Front End=0
    )
    Use Spawn Probability=false
   ) {module=SpawnRate003}
   &SpawnRate(
    Spawn Probability=&RandomRangeFloat(Minimum=0.3, Maximum=2)
    SpawnGroup=4
    SpawnRate=&RandomRangeFloat(Minimum=0.5, Maximum=6)
    Use Spawn Probability=true
   ) {module=SpawnRate002}
   &SpawnRate(SpawnGroup=3, SpawnRate=&RandomRangeFloat(Minimum=0.5, Maximum=2), Use Spawn Probability=false) {module=SpawnRate001}
   &SpawnRate(
    Spawn Probability=&RandomRangeFloat(Minimum=0.5, Maximum=1, Randomness Mode=::NewEnumerator2, Random Seed=?)
    SpawnGroup=2
    SpawnRate=2
    Use Spawn Probability=true
   )
   &UpdateValueStateChange(InputIntParam:NiagaraInt32=&MakeIntFromFloat(Float=$User.BurstIndex, Conversion Mode=::NewEnumerator1))
   &self:TriggerSpawn_16(
    Spawn Count:NiagaraInt32=&MergeHelper_IntSwitch(
     0=1
     1=1
     2=1
     3=1
     4=1
     5=1
     6=1
     7=1
     8=1
     Index=&MakeIntFromFloat(Float=$User.BurstIndex, Conversion Mode=::NewEnumerator1)
     Num=::NewEnumerator7
    )
    Spawn Group:NiagaraInt32=&MakeIntFromFloat(Float=0, Conversion Mode=::NewEnumerator1)
    Trigger:NiagaraBool=$Emitter.UpdateValueStateChange.Changed
   ) {module=TriggerSpawn001}
   &self:TriggerSpawn_17(
    Spawn Count:NiagaraInt32=&MergeHelper_IntSwitch(
     0=1
     1=1
     2=1
     3=1
     4=1
     5=1
     6=1
     7=1
     8=1
     Index=&MakeIntFromFloat(Float=$User.BurstIndex, Conversion Mode=::NewEnumerator1)
     Num=::NewEnumerator7
    )
    Spawn Group:NiagaraInt32=&MakeIntFromFloat(Float=1, Conversion Mode=::NewEnumerator1)
    Trigger:NiagaraBool=$Emitter.UpdateValueStateChange.Changed
   ) {module=TriggerSpawn_17}
   &UpdateValueStateChange(InputIntParam:NiagaraInt32=&MakeIntFromFloat(Float=$User.BurstIndex, Conversion Mode=::NewEnumerator1)) {module=UpdateValueStateChange001}
   &self:TriggerSpawn_17(
    Spawn Count:NiagaraInt32=&MergeHelper_IntSwitch(
     0=1
     1=1
     2=1
     3=1
     4=1
     5=1
     6=1
     7=1
     8=1
     Index=&MakeIntFromFloat(Float=$User.BurstIndex, Conversion Mode=::NewEnumerator1)
     Num=::NewEnumerator7
    )
    Spawn Group:NiagaraInt32=&MakeIntFromFloat(Float=5, Conversion Mode=::NewEnumerator1)
    Trigger:NiagaraBool=$Emitter.UpdateValueStateChange001.Changed
   ) {module=TriggerSpawn002}
  ]
  particleSpawnScript [
   /SetVariables(Particles.VisibilityTag=&SpawnGroup()) {isSetParameters=true}
   &InitializeParticle(
    Color=&FNTimeOfDayColor(Night Color=<rgba 0 0.0009 0.0156 1>, Morning Color=<rgba 0 0 0 1>, Day Color=<rgba 0 0 0 1>, Evening Color=<rgba 0 0 0 1>)
    Lifetime=&MergeHelper_FloatSwitch(
     0=&FN_FrontEndFloatModifier(
      Gameplay=&MakeCustomFloatFromBool(
       Bool=&SetBoolByFloatComparison(A=&VectorLength(Vector=$User.OwnerLocalVelocity), B=139, Comparison Type=::NewEnumerator6)
       True Float=$Emitter.Idle Lifespan
       False Float=&MakeCustomFloatFromBool(
        Bool=&SetBoolByFloatComparison(A=&VectorLength(Vector=$User.OwnerLocalVelocity), B=500, Comparison Type=::NewEnumerator6)
        True Float=$Emitter.Traversal LifeSpawn
        False Float=Linked() {linkedVariable {name=Emitter.Car/rail Lifespawn, type=NiagaraFloat}}
       )
      )
      Front End=$Emitter.Traversal LifeSpawn
     )
     1=&FN_FrontEndFloatModifier(
      Gameplay=&FN_FrontEndFloatModifier(
       Gameplay=&MakeCustomFloatFromBool(
        Bool=&SetBoolByFloatComparison(A=&VectorLength(Vector=$User.OwnerLocalVelocity), B=139, Comparison Type=::NewEnumerator6)
        True Float=0
        False Float=&MakeCustomFloatFromBool(
         Bool=&SetBoolByFloatComparison(A=&VectorLength(Vector=$User.OwnerLocalVelocity), B=500, Comparison Type=::NewEnumerator6)
         True Float=$Emitter.Traversal lifespawn decal 2
         False Float=Linked() {linkedVariable {name=Emitter.Car/rail Lifespawn, type=NiagaraFloat}}
        )
       )
       Front End=$Emitter.Traversal lifespawn decal 2
      )
      Front End=1
     )
     2=&FN_FrontEndFloatModifier(
      Gameplay=&MakeCustomFloatFromBool(
       Bool=&SetBoolByFloatComparison(A=&VectorLength(Vector=$User.OwnerLocalVelocity), B=139, Comparison Type=::NewEnumerator6)
       True Float=$Emitter.Heart Traversal Lifespawn
       False Float=&MakeCustomFloatFromBool(
        Bool=&SetBoolByFloatComparison(A=&VectorLength(Vector=$User.OwnerLocalVelocity), B=500, Comparison Type=::NewEnumerator6)
        True Float=$Emitter.Heart Traversal Lifespawn
        False Float=$Emitter.Icon on rail Lifespawn
       )
      )
      Front End=$Emitter.Heart Traversal Lifespawn
     )
     3=&FN_FrontEndFloatModifier(
      Gameplay=&MakeCustomFloatFromBool(
       Bool=&SetBoolByFloatComparison(A=&VectorLength(Vector=$User.OwnerLocalVelocity), B=139, Comparison Type=::NewEnumerator6)
       True Float=Linked() {linkedVariable {name="Emitter.Meow  Traversal Lifespawn", type=NiagaraFloat}}
       False Float=&MakeCustomFloatFromBool(
        Bool=&SetBoolByFloatComparison(A=&VectorLength(Vector=$User.OwnerLocalVelocity), B=500, Comparison Type=::NewEnumerator6)
        True Float=Linked() {linkedVariable {name="Emitter.Meow  Traversal Lifespawn", type=NiagaraFloat}}
        False Float=$Emitter.Icon on rail Lifespawn
       )
      )
      Front End=Linked() {linkedVariable {name="Emitter.Meow  Traversal Lifespawn", type=NiagaraFloat}}
     )
     4=&FN_FrontEndFloatModifier(
      Gameplay=&MakeCustomFloatFromBool(
       Bool=&SetBoolByFloatComparison(A=&VectorLength(Vector=$User.OwnerLocalVelocity), B=139, Comparison Type=::NewEnumerator6)
       True Float=$Emitter.Star Traversal Lifespawn
       False Float=&MakeCustomFloatFromBool(
        Bool=&SetBoolByFloatComparison(A=&VectorLength(Vector=$User.OwnerLocalVelocity), B=500, Comparison Type=::NewEnumerator6)
        True Float=$Emitter.Star Traversal Lifespawn
        False Float=$Emitter.Icon on rail Lifespawn
       )
      )
      Front End=$Emitter.Star Traversal Lifespawn
     )
     5=&FN_FrontEndFloatModifier(
      Gameplay=&FN_FrontEndFloatModifier(
       Gameplay=&MakeCustomFloatFromBool(
        Bool=&SetBoolByFloatComparison(A=&VectorLength(Vector=$User.OwnerLocalVelocity), B=139, Comparison Type=::NewEnumerator6)
        True Float=$Emitter.Idle Lifespawn decal2
        False Float=&MakeCustomFloatFromBool(
         Bool=&SetBoolByFloatComparison(A=&VectorLength(Vector=$User.OwnerLocalVelocity), B=500, Comparison Type=::NewEnumerator6)
         True Float=$Emitter.Traversal lifespawn decal 2
         False Float=Linked() {linkedVariable {name=Emitter.Car/rail Lifespawn, type=NiagaraFloat}}
        )
       )
       Front End=$Emitter.Traversal lifespawn decal 2
      )
      Front End=1
     )
     6=&RandomRangeFloat(Minimum=0.85, Maximum=1)
     Index=&SpawnGroup()
     Num=::NewEnumerator5
    )
    Mesh Scale Mode=::NewEnumerator3
    Mesh Uniform Scale=3
    Sprite Rotation Mode=::NewEnumerator2
    Sprite Size=&MergeHelper_Vector2Switch(
     0=&Vector2DFromFloat(Value=60)
     1=&Vector2DFromFloat(Value=60)
     2=&Vector2DFromFloat(Value=50)
     3=&Vector2DFromFloat(Value=50)
     4=<0 0>
     Index=&SpawnGroup()
     Num=::NewEnumerator3
    )
    Sprite Size Mode=::NewEnumerator4
    UsePositionOffset=true
    Write Material Random=true
   )
   &SkeletalMeshLocation(
    Bone / Socket Index=&MakeIntFromFloat(Float=$User.BurstIndex, Conversion Mode=::NewEnumerator1)
    Bone Sampling Mode=::NewEnumerator4
    MaskSpawnGroup=false
    Orientation Sampling=::NewEnumerator1
    Sampled Position Offset=&MergeHelper_VectorSwitch(
     0=&FN_FrontEndVector3Modifier(Gameplay=<0 0 0>, Front End=<0 -5 -8>)
     1=&FN_FrontEndVector3Modifier(Gameplay=<-3 0 0>, Front End=<-3 -5 -8>)
     2=?
     3=?
     Index=&MakeIntFromFloat(Float=$User.BurstIndex, Conversion Mode=::NewEnumerator1)
     Num=::NewEnumerator0
    )
    Sampled Position Offset Coordinate Space:ENiagaraSkelMeshTransforms=::NewEnumerator2
    Skeletal Mesh=$System.Skeletal Mesh
   )
   &SpriteFacingAndAlignment(Sprite Facing=<0 0 1>, WriteAlignment=true, WriteFacing=true)
   &CameraOffset(Camera Offset Amount=&MergeHelper_FloatSwitch(
     0=-2
     1=-2
     2=&RandomRangeFloat(Minimum=-5, Maximum=5)
     3=&RandomRangeFloat(Minimum=-5, Maximum=5)
     4=&RandomRangeFloat(Minimum=-5, Maximum=5)
     Index=&SpawnGroup()
     Num=::NewEnumerator3
    ))
   /SetVariables(Particles.heart Position:NiagaraPosition=$Particles.SkeletalMeshLocation.SampledPosition) {isSetParameters=true}
   /SetVariables(Particles.Meow Position:NiagaraPosition=$Particles.SkeletalMeshLocation.SampledPosition) {isSetParameters=true}
   /SetVariables(Particles.Star Position:NiagaraPosition=$Particles.SkeletalMeshLocation.SampledPosition) {isSetParameters=true}
  ]
  particleUpdateScript [
   &ParticleState()
   &Drag(
    Drag=&MakeCustomFloatFromBool(
     Bool=&SetBoolByFloatComparison(A=&VectorLength(Vector=$User.OwnerLocalVelocity), B=139, Comparison Type=::NewEnumerator6)
     True Float=$Emitter.Idle Drag
     False Float=&MakeCustomFloatFromBool(
      Bool=&SetBoolByFloatComparison(A=&VectorLength(Vector=$User.OwnerLocalVelocity), B=500, Comparison Type=::NewEnumerator6)
      True Float=$Emitter.Traversal Drag
      False Float=$Emitter.Rail drag
     )
    )
    Use Linear Drag=true
    Use Rotational Drag=true
   )
   &GravityForce(
    Coordinate Space=::Local
    Gravity=&MergeHelper_VectorSwitch(
     0=<0 0 0>
     1=<0 0 0>
     2=&MakeVector(
      X=&Lerp_Float(A=-250, B=250, Alpha=$User.KickbackBurst)
      Y=&FloatFromCurve(FloatCurve=DataInterface() {propertyValues {Curve {keys [<at 0 -350>,<at 0.5 -190>]}}})
      Z=&FloatFromCurve(FloatCurve=DataInterface() {propertyValues {Curve {keys [<at 0 0>,<at 0.2 950>,<at 0.4976 -450>,<at 0.75 650>,<at 1 -250>]}}})
     )
     3=&MakeVector(
      X=&Lerp_Float(A=200, B=-200, Alpha=$User.KickForward_Idle)
      Y=&FloatFromCurve(FloatCurve=DataInterface() {propertyValues {Curve {keys [<at 0 -350>,<at 0.5 -50>]}}}, Scale Curve=0.5)
      Z=&FloatFromCurve(FloatCurve=DataInterface() {propertyValues {Curve {keys [<at 0 0>,<at 0.1192 950>,<at 0.75 650>,<at 1 -250>]}}}, Scale Curve=0.45)
     )
     4=&MakeVector(
      X=&Lerp_Float(A=200, B=-200, Alpha=$User.KickForward_Idle)
      Y=&FloatFromCurve(FloatCurve=DataInterface() {propertyValues {Curve {keys [<at 0 -350>,<at 0.5 -50>]}}}, Scale Curve=0.5)
      Z=&FloatFromCurve(FloatCurve=DataInterface() {propertyValues {Curve {keys [<at 0 0>,<at 0.1192 950>,<at 0.75 650>,<at 1 -250>]}}}, Scale Curve=0.45)
     )
     Index=&SpawnGroup()
     Num=::NewEnumerator3
    )
   )
   &SubUVAnimation(
    Define SubUV Setup Manually:ENiagara_AutomaticManual=::NewEnumerator1
    Direct Frame Index:NiagaraFloat=&SelectFloatFromArray(
     Array Sampling Mode=::NewEnumerator2
     Direct Array Index=?
     Interpolation=$Particles.NormalizedAge
     Float Selection Array=DataInterface() {propertyValues {FloatData [1,2,3,4,5,5]}}
    )
    End Frame Range Override=10
    Output Alternate Renderer Binding=true
    SubUV Animation Mode:ENiagara_SubUVLookupModeV2=::NewEnumerator5
    UseEndFrame=true
    UseStartFrame=false
   ) {module=SubUVAnimation002}
   /SetVariables(
    Particles.Feet Splat DynParam 0:Vector4f=&MakeVector4(X=0, Y=0, Z=1, W=1)
    Particles.Feet Splat DynParam 1:Vector4f=&MakeVector4(X=0, Y=0, Z=1, W=1)
   ) {isSetParameters=true}
   &OffsetPosition(
    Position Offset=&MakeVector(
     X=0
     Y=&MergeHelper_FloatSwitch(
      0=&FloatFromCurve(
       FloatCurve=DataInterface() {propertyValues {Curve {defaultValue=1, keys [{tangentMode=RCTM_User, time=0, value=0}, {tangentMode=RCTM_User, time=1, value=1}]}}}
       Scale Curve=&FN_FrontEndFloatModifier(Gameplay=-100, Front End=-15)
      )
      1=&FloatFromCurve(
       FloatCurve=DataInterface() {propertyValues {Curve {defaultValue=1, keys [{tangentMode=RCTM_User, time=0, value=0}, {tangentMode=RCTM_User, time=1, value=1}]}}}
       Scale Curve=&FN_FrontEndFloatModifier(Gameplay=-15, Front End=-100)
      )
      2=0
      3=0
      4=0
      5=&FloatFromCurve(
       FloatCurve=DataInterface() {propertyValues {Curve {defaultValue=1, keys [{tangentMode=RCTM_User, time=0, value=0}, {tangentMode=RCTM_User, time=1, value=1}]}}}
       Scale Curve=&FN_FrontEndFloatModifier(Gameplay=-15, Front End=-100)
      )
      6=0
      Index=&SpawnGroup()
      Num=::NewEnumerator5
     )
     Z=0
    )
    Position Offset Coordinate Space:ENiagaraCoordinateSpace=::Local
   )
   /SetVariables(Particles.DecalSize:Vector3f=&Multiply_Vector(
     A=&VectorFromFloat(Value=&RandomRangeFloat(Minimum=5, Maximum=27.5))
     B=&VectorFromFloat(Value=&FloatFromCurve(
       FloatCurve=DataInterface() {propertyValues {Curve {keys [
           {arriveTangent=4.1785, interpMode=RCIM_Cubic, leaveTangent=4.1785, tangentMode=RCTM_User, time=0, value=0.5}
           {arriveTangent=1.1179, interpMode=RCIM_Cubic, leaveTangent=1.1179, tangentMode=RCTM_User, time=0.01, value=0.8}
           {arriveTangent=0.0224, interpMode=RCIM_Cubic, leaveTangent=0.0224, tangentMode=RCTM_User, time=0.05, value=1}
          ]}}}
       Scale Curve=&RandomRangeFloat(Minimum=0.7, Maximum=2.5)
      ))
    )) {isSetParameters=true}
   /SetVariables(Particles.DecalFade:NiagaraFloat=&FloatFromCurve(FloatCurve=DataInterface() {propertyValues {Curve {keys [<at 0 1>,<at 1 0>]}}})) {isSetParameters=true}
   /SetVariables(
    Particles.Layer 1 DynParam 0:Vector4f=&MakeVector4(
     X=&FloatFromCurve(FloatCurve=DataInterface() {propertyValues {Curve {keys [
          {interpMode=RCIM_Cubic, tangentMode=RCTM_Break, time=0.0356, value=0.9644}
          {interpMode=RCIM_Cubic, tangentMode=RCTM_Break, time=0.8467, value=0.8}
          {interpMode=RCIM_Cubic, tangentMode=RCTM_Break, time=1, value=0}
         ]}}})
     Y=&FloatFromCurve(FloatCurve=DataInterface() {propertyValues {Curve {keys [
          {interpMode=RCIM_Cubic, tangentMode=RCTM_Break, time=0.0356, value=0.9644}
          {interpMode=RCIM_Cubic, tangentMode=RCTM_Break, time=0.8467, value=0.8}
          {interpMode=RCIM_Cubic, tangentMode=RCTM_Break, time=1, value=0}
         ]}}})
     Z=0
     W=0
    )
    Particles.Layer 1 DynParam 1:Vector4f=&MakeVector4(X=1, Y=1, Z=0, W=0)
   ) {isSetParameters=true}
   &self:Slope_Position_18(
    Apply Position:NiagaraBool=&InvertBool(Boolean:NiagaraBool=$User.bIsFrontEndPreview)
    Apply Rotation:NiagaraBool=&InvertBool(Boolean:NiagaraBool=$User.bIsFrontEndPreview)
    Fallback Position:NiagaraPosition=&AddVectorToPosition(Position=$Particles.Initial.Position, Vector=<1 1 1>)
    Height Adjustment/Offset=0.5
    Slope Position:NiagaraPosition=$Particles.SkeletalMeshLocation.SampledPosition
    Trace Length=120
    Trace Origin Height=85
   ) {module=Slope_Position_18}
   /SetVariables(
    Particles.Fake Decal SpriteSize:Vector2f=&Multiply_Vector2D_ByFloat(
     Float=&FloatFromCurve(
      FloatCurve=DataInterface() {propertyValues {Curve {
         defaultValue=1
         keys [
          {interpMode=RCIM_Cubic, tangentMode=RCTM_Break, time=0, value=0}
          {interpMode=RCIM_Cubic, tangentMode=RCTM_Break, time=0.1781, value=1.1}
          <at 0.1958 1 RCIM_Cubic>
         ]
        }}}
      Scale Curve=&RandomRangeFloat(Minimum=0.85, Maximum=1.3)
     )
     Vector2D=<50 50>
    )
    Particles.Fake decal SpriteRotation:NiagaraFloat=&RandomRangeFloat(Minimum=-1, Maximum=1)
    Particles.Fake decal color:LinearColor=&MakeLinearColorFromVectorAndFloat(
     "Vector (RGB)"=<1 1 1>
     "Float (Alpha)"=&FloatFromCurve(FloatCurve=DataInterface() {propertyValues {Curve {keys [<at 0 0>,<at 0.15 0.9526>,<at 0.8005 1>,<at 1 0>]}}})
    )
   ) {isSetParameters=true}
   &SubUVAnimation(
    End Frame Range Override=10
    Output Alternate Renderer Binding=true
    "Sprite Renderer "=DataInterface() {propertyValues {SpriteRenderer=&self:foot_impact_w_slope_detection001_0.NiagaraSpriteRendererProperties_7}}
    SubUV Lookup Index:NiagaraFloat=&FloatFromCurve(FloatCurve=DataInterface() {propertyValues {Curve {defaultValue=1, keys [<at 0 0>,<at 0.5 1>,<at 1 0>]}}})
    UseEndFrame=true
    UseStartFrame=true
   ) {module=SubUVAnimation004}
   &SubUVAnimation(
    End Frame Range Override=11
    Output Alternate Renderer Binding=true
    Random Start Frame=true
    "Sprite Renderer "=DataInterface() {propertyValues {SpriteRenderer=&self:foot_impact_w_slope_detection001_0.NiagaraSpriteRendererProperties_6}}
    SubUV Lookup Index:NiagaraFloat=&FloatFromCurve(FloatCurve=DataInterface() {propertyValues {Curve {defaultValue=0, keys [<at 0 0>,<at 0.5 1>,<at 1 0>]}}})
    UseEndFrame=true
    UseStartFrame=true
   ) {module=SubUVAnimation005}
   /SetVariables(
    Particles.Heart DynParam 0:Vector4f=&MakeVector4(X=1, Y=1, Z=0, W=0)
    Particles.Heart DynParam 1:Vector4f=&MakeVector4(X=1, Y=1, Z=0, W=0)
    Particles.Heart SpriteSize:Vector2f=&Multiply_Vector2D_ByFloat(
     Float=&FloatFromCurve(
      FloatCurve=DataInterface() {propertyValues {Curve {keys [
          <at 0 0 RCIM_Cubic>
          {arriveTangent=-0.0027, interpMode=RCIM_Cubic, leaveTangent=-0.0027, tangentMode=RCTM_User, time=0.0604, value=1.3}
          {arriveTangent=-0.4266, interpMode=RCIM_Cubic, leaveTangent=-0.4266, time=0.1204, value=1}
          {arriveTangent=-0.2399, interpMode=RCIM_Cubic, leaveTangent=-0.2399, tangentMode=RCTM_User, time=0.7636, value=1}
          {arriveTangent=-9.7345, interpMode=RCIM_Cubic, leaveTangent=-9.7345, tangentMode=RCTM_User, time=0.8449, value=1.5}
          <at 1 0 RCIM_Cubic>
         ]}}}
      Scale Curve=&RandomRangeFloat(Minimum=0.7, Maximum=1)
     )
     Vector2D=<20 20>
    )
   ) {isSetParameters=true}
   /SetVariables(Particles.heart Position:NiagaraPosition=&AddVectorToPosition(
     Position=$Particles.heart Position
     Vector=&Multiply_Vector(A=$Particles.Velocity, B=&VectorFromFloat(Value=$Engine.DeltaTime))
    )) {isSetParameters=true}
   /SetVariables(
    Particles.Meow DynParam 0:Vector4f=&MakeVector4(X=0, Y=0, Z=1, W=1)
    Particles.Meow DynParam 1:Vector4f=&MakeVector4(X=0, Y=0, Z=1, W=1)
    Particles.Meow SpriteSize:Vector2f=&Multiply_Vector2D_ByFloat(
     Float=&FloatFromCurve(
      FloatCurve=DataInterface() {propertyValues {Curve {keys [
          <at 0 0 RCIM_Cubic>
          {arriveTangent=-0.0027, interpMode=RCIM_Cubic, leaveTangent=-0.0027, tangentMode=RCTM_User, time=0.0604, value=1.3}
          {arriveTangent=-0.4266, interpMode=RCIM_Cubic, leaveTangent=-0.4266, time=0.1204, value=1}
          {arriveTangent=-0.2399, interpMode=RCIM_Cubic, leaveTangent=-0.2399, tangentMode=RCTM_User, time=0.7636, value=1}
          {arriveTangent=-9.7345, interpMode=RCIM_Cubic, leaveTangent=-9.7345, tangentMode=RCTM_User, time=0.8449, value=1.5}
          <at 1 0 RCIM_Cubic>
         ]}}}
      Scale Curve=&RandomRangeFloat(Minimum=1, Maximum=1.7)
     )
     Vector2D=<20 20>
    )
   ) {isSetParameters=true}
   /SetVariables(Particles.Meow Position:NiagaraPosition=&AddVectorToPosition(
     Position=$Particles.Meow Position
     Vector=&Multiply_Vector(A=$Particles.Velocity, B=&VectorFromFloat(Value=$Engine.DeltaTime))
    )) {isSetParameters=true}
   /SetVariables(
    Particles.Star Color:LinearColor=&MakeLinearColorFromVectorAndFloat(
     "Vector (RGB)"=&MakeVectorFromLinearColorRGB(Color=&FNTimeOfDayColor(Night Color=<rgba 0 0.0003 0.0015 1>, Morning Color=<rgba 0 0 0 1>, Day Color=<rgba 0 0 0 1>, Evening Color=<rgba 0 0 0 1>))
     "Float (Alpha)"=1
    )
    Particles.Star Position:NiagaraPosition=&AddVectorToPosition(
     Position=$Particles.Star Position
     Vector=&Multiply_Vector(A=$Particles.Velocity, B=&VectorFromFloat(Value=$Engine.DeltaTime))
    )
    Particles.Star SpriteSize:Vector2f=&Multiply_Vector2D_ByFloat(
     Float=&FloatFromCurve(
      FloatCurve=DataInterface() {propertyValues {Curve {keys [
          <at 0 0 RCIM_Cubic>
          {arriveTangent=-0.0027, interpMode=RCIM_Cubic, leaveTangent=-0.0027, tangentMode=RCTM_User, time=0.0604, value=1.3}
          {arriveTangent=-0.4266, interpMode=RCIM_Cubic, leaveTangent=-0.4266, time=0.1204, value=1}
          {arriveTangent=-0.2399, interpMode=RCIM_Cubic, leaveTangent=-0.2399, tangentMode=RCTM_User, time=0.7636, value=1}
          {arriveTangent=-9.7345, interpMode=RCIM_Cubic, leaveTangent=-9.7345, tangentMode=RCTM_User, time=0.8449, value=1.5}
          <at 1 0 RCIM_Cubic>
         ]}}}
      Scale Curve=&RandomRangeFloat(Minimum=0.5, Maximum=1)
     )
     Vector2D=<20 20>
    )
   ) {isSetParameters=true}
   /SetVariables(Particles.Icons SpriteRotation:NiagaraFloat=&Lerp_Float(A=-20, B=50, Alpha=$Particles.NormalizedAge)) {isSetParameters=true}
   &SolveForcesAndVelocity(Clamp Velocity=false, Limit Acceleration=false)
   &BlendSpace(
    "Position Blend (World <-> Local)":NiagaraFloat=&MakeCustomFloatFromBool(
     Bool=&SetBoolByFloatComparison(A=&VectorLength(Vector=$Engine.Owner.Velocity), B=$Emitter.Max Traversal speed)
     True Float=1
     False Float=&FloatFromCurve(FloatCurve=DataInterface() {propertyValues {Curve {
         defaultValue=0
         keys [{interpMode=RCIM_Cubic, tangentMode=RCTM_User, time=0, value=1}, {interpMode=RCIM_Cubic, tangentMode=RCTM_User, time=0.15, value=0}]
        }}})
     Translation Contribution:NiagaraFloat=0
     Rotation Contribution:NiagaraFloat=0
    )
    Position Coordinate Space=2
   ) {module=BlendSpace002}
   &ScaleColor(
    Scale Alpha=&FloatFromCurve(FloatCurve=DataInterface() {propertyValues {Curve {keys [
         {interpMode=RCIM_Cubic, tangentMode=RCTM_Break, time=0, value=1}
         {interpMode=RCIM_Cubic, tangentMode=RCTM_Break, time=0.8155, value=1}
         {interpMode=RCIM_Cubic, tangentMode=RCTM_Break, time=1, value=0}
        ]}}})
    ScaleA=true
    ScaleRGB=true
   )
   &Decal_Attributes(
    Color=1
    Decal Facing Mode=1
    Decal Rotation Mode=1
    Fade=1
    Orientation=1
    Scale RGB:Vector3f=&Multiply_VectorByFloat(Vector=<1 1 1>, Float=1)
    ScaleA=true
    ScaleRGB=true
    Set Alignment=false
   )
   /SetVariables(Particles.DecalSize2:Vector3f=&Multiply_Vector(
     A=&VectorFromFloat(Value=&RandomRangeFloat(Minimum=10, Maximum=15))
     B=&VectorFromFloat(Value=&FloatFromCurve(
       FloatCurve=DataInterface() {propertyValues {Curve {keys [
           {arriveTangent=4.1785, interpMode=RCIM_Cubic, leaveTangent=4.1785, tangentMode=RCTM_User, time=0, value=0.5}
           {arriveTangent=11.9214, interpMode=RCIM_Cubic, leaveTangent=11.9214, tangentMode=RCTM_User, time=0.01, value=0.8}
           {arriveTangent=0.0224, interpMode=RCIM_Cubic, leaveTangent=0.0224, tangentMode=RCTM_User, time=0.05, value=1}
          ]}}}
       Scale Curve=&RandomRangeFloat(Minimum=1, Maximum=3)
      ))
    )) {isSetParameters=true}
  ]
  rendererClasses [&NiagaraSpriteRendererProperties, &NiagaraDecalRendererProperties]
  renderers [
   {rendererClass=&NiagaraSpriteRendererProperties}
   {rendererClass=&NiagaraSpriteRendererProperties, rendererIndex=1}
   {rendererClass=&NiagaraSpriteRendererProperties, rendererIndex=2}
   {rendererClass=&NiagaraSpriteRendererProperties, rendererIndex=3}
   {rendererClass=&NiagaraSpriteRendererProperties, rendererIndex=4}
   {rendererClass=&NiagaraSpriteRendererProperties, rendererIndex=5}
   {rendererClass=&NiagaraSpriteRendererProperties, rendererIndex=6}
   {rendererClass=&NiagaraSpriteRendererProperties, rendererIndex=7}
   {rendererClass=&NiagaraDecalRendererProperties, rendererIndex=8}
   {rendererClass=&NiagaraDecalRendererProperties, rendererIndex=9}
  ]
 }
}
summary {
 systemName=NS_Bombatoon_traversal_footImpacts_looping
 emitters [
  {emitterName=Decal002, rendererClasses [&NiagaraDecalRendererProperties, &NiagaraSpriteRendererProperties]}
  {emitterName=GOOP_STRETCH, rendererClasses [&NiagaraMeshRendererProperties, &NiagaraSpriteRendererProperties]}
  {emitterName=Foot_FX_merged, rendererClasses [&NiagaraMeshRendererProperties, &NiagaraSpriteRendererProperties]}
  {
   emitterName=foot_impact_w_slope_detection001
   rendererClasses [&NiagaraSpriteRendererProperties, &NiagaraDecalRendererProperties]
  }
 ]
 userVariables [
  User.bIsFrontEnd {type=NiagaraBool}
  User.bIsFrontEndPreview {type=NiagaraBool}
  User.bisHeadHidden {type=NiagaraBool}
  User.BurstIndex {type=NiagaraFloat, defaultValue=0, description=""}
  User.KickbackBurst {type=NiagaraFloat, defaultValue=0, description=""}
  User.KickForward_Idle {type=NiagaraFloat, defaultValue=1, description=""}
  User.OwnerLocalVelocity {type=Vector3f, description="", defaultValue {struct=Vector3f, value=<0 0 0>}}
  User.splashBurst {type=NiagaraFloat, defaultValue=0, description=""}
 ]
}
